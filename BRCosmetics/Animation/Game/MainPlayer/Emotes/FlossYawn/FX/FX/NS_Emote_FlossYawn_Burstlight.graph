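# Niagara system for a burst-light effect on the FlossYawn emote, built from an expanding mesh ring, ring-placed stretched mesh streaks and a spinning sprite flash.
# All three use instantaneous bursts and blue-to-cyan tints with curve-driven fades.
# Front-end, platform and time-of-day functions scale some alpha and glow.
# These detailed emitters are disabled; their 001 counterparts show no modules or renderers.
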NiagaraSystem /BRCosmetics/Animation/Game/MainPlayer/Emotes/FlossYawn/FX/FX/NS_Emote_FlossYawn_Burstlight
dependencies {
 usedDataInterfaces [&NiagaraDataInterfaceColorCurve, &NiagaraDataInterfaceCurve, &NiagaraDataInterfacePlatformSet]
 usedDynamicInputs []
 usedModules [&SystemState]
}
emitters {
 Burst_Ring {
  bEnabled=false
  emitterUpdateScript [
   &EmitterState(Inactive Response=::NewEnumerator1, Life Cycle Mode=::NewEnumerator1, Loop Behavior=::NewEnumerator1, UseLoopDelay=false)
   &SpawnBurst_Instantaneous(Spawn Count=1, Use Spawn Probability=true)
  ]
  particleSpawnScript [
   &InitializeParticle(
    Color=<rgba 1 1 1 0.25>
    Lifetime=0.8
    Mesh Scale Mode=::NewEnumerator3
    Mesh Uniform Scale=1.5
    Position Offset=<0 0 10>
    Sprite Size Mode=::NewEnumerator3
    Uniform Sprite Size=30
    UsePositionOffset=true
    Write Material Random=true
   ) {module=InitializeParticle001}
   &InitialMeshOrientation(Orientation Axis=<0 0 1>, Orientation Vector=<0 0 1>, Rotation=<0 0 0.25>, Use Orientation Vector=true, Use Rotation Vector=true)
   &InitialMeshOrientation(
    Orientation Axis=<0 0 0>
    Orientation Vector=<0 0 0>
    Rotation=&RandomRangeVector(Minimum=<0 0 0>, Maximum=<0 0 0>)
    Use Orientation Vector=false
    Use Rotation Vector=true
   ) {enabled=false, module=InitialMeshOrientation001}
   /SetVariables() {isSetParameters=true}
   /SetVariables(Particles.Position002:NiagaraPosition=<0 0 20>) {isSetParameters=true}
   &SpriteFacingAndAlignment(Sprite Facing=<0 0 1>, WriteAlignment=true, WriteFacing=true)
  ]
  particleUpdateScript [
   &ParticleState()
   &ScaleColor(
    Scale Alpha=&FloatFromCurve(
     FloatCurve=DataInterface() {propertyValues {Curve {keys [
         {arriveTangent=-3.8906, interpMode=RCIM_Cubic, leaveTangent=-3.8906, tangentMode=RCTM_User, time=0, value=1}
         {arriveTangent=-1.0522, interpMode=RCIM_Cubic, leaveTangent=-1.0522, tangentMode=RCTM_User, time=0.2, value=0.4}
         <at 1 0 RCIM_Cubic>
        ]}}}
     Scale Curve=&FN_PickFloatDependOnPlatform(TargetPlatformActive=0.5, TargetPlatformNoActive=0.1)
    )
    Scale RGB=&Multiply_Vector(
     A=&MakeVectorFromLinearColorRGB(Color=&ColorFromCurve(ColorCurve=DataInterface() {propertyValues {
         AlphaCurve {keys []}
         BlueCurve {keys [<at 0.1142 1 RCIM_Cubic>,<at 0.259 1>,<at 0.6726 1>]}
         GreenCurve {keys [<at 0.1142 0.4033 RCIM_Cubic>,<at 0.259 0.6419>,<at 0.6726 0.4111>]}
         RedCurve {keys [<at 0.1142 0.1504 RCIM_Cubic>,<at 0.259 0.2394>,<at 0.6726 0>]}
        }}))
     B=&VectorFromFloat(Value=&FloatFromCurve(
       FloatCurve=DataInterface() {propertyValues {Curve {keys [{arriveTangent=-1.3603, interpMode=RCIM_Cubic, leaveTangent=-1.3603, tangentMode=RCTM_User, time=0, value=0.8}, <at 1 0 RCIM_Cubic>]}}}
       Scale Curve=0.4
      ))
    )
    ScaleA=true
    ScaleRGB=true
   ) {module=ScaleColor001}
   &ScaleMeshSize(Scale Factor=&VectorFromFloat(Value=&FloatFromCurve(FloatCurve=DataInterface() {propertyValues {Curve {keys [
          {arriveTangent=4.526, interpMode=RCIM_Cubic, leaveTangent=4.526, tangentMode=RCTM_User, time=0, value=0}
          {arriveTangent=2, interpMode=RCIM_Cubic, leaveTangent=2, time=0.3, value=1.25}
          <at 1 2 RCIM_Cubic>
         ]}}})))
   /SetVariables(Particles.GlowColor=&MakeLinearColorFromVectorAndFloat(
     "Vector (RGB)"=&MakeVectorFromLinearColorRGB(Color=&ColorFromCurve(ColorCurve=DataInterface() {propertyValues {
         AlphaCurve {keys []}
         BlueCurve {keys [<at 0.1724 1 RCIM_Cubic>,<at 0.365 1>,<at 0.5023 1>,<at 0.7173 1>,<at 0.8592 1 RCIM_Cubic>]}
         GreenCurve {keys [<at 0.1724 0.6514 RCIM_Cubic>,<at 0.365 0.4115>,<at 0.5023 0.442>,<at 0.7173 0.6038>,<at 0.8592 0.9911 RCIM_Cubic>]}
         RedCurve {keys [<at 0.1724 0.2705 RCIM_Cubic>,<at 0.365 0.1793>,<at 0.5023 0.229>,<at 0.7173 0.3185>,<at 0.8592 0.4179 RCIM_Cubic>]}
        }}))
     "Float (Alpha)"=&FloatFromCurve(
      FloatCurve=DataInterface() {propertyValues {Curve {keys [
          {arriveTangent=-1.8164, interpMode=RCIM_Cubic, leaveTangent=-1.8164, tangentMode=RCTM_User, time=0, value=0.4}
          {arriveTangent=-0.0068, interpMode=RCIM_Cubic, leaveTangent=-0.0068, tangentMode=RCTM_User, time=0.1, value=0.8}
          {arriveTangent=-0.11, interpMode=RCIM_Cubic, leaveTangent=-0.11, tangentMode=RCTM_User, time=0.7, value=0}
         ]}}}
      Scale Curve=&FN_FrontEndFloatModifier(Gameplay=0.2, Front End=0.3)
     )
    )) {isSetParameters=true}
   /SetVariables(Particles.Size002:Vector3f=&Multiply_Vector(
     A=&VectorFromFloat(Value=&FloatFromCurve(FloatCurve=DataInterface() {propertyValues {Curve {keys [
           {arriveTangent=6.412, interpMode=RCIM_Cubic, leaveTangent=6.412, tangentMode=RCTM_User, time=0, value=0}
           {arriveTangent=2, interpMode=RCIM_Cubic, leaveTangent=2, time=0.1633, value=0.7527}
           {arriveTangent=0.3953, interpMode=RCIM_Cubic, leaveTangent=0.3953, tangentMode=RCTM_User, time=0.6, value=1.2}
          ]}}}))
     B=&VectorFromFloat(Value=1)
    )) {isSetParameters=true}
   &DynamicMaterialParameters(
    Index 0 Param 1=&FloatFromCurve(FloatCurve=DataInterface() {propertyValues {Curve {keys [{arriveTangent=-0.466, interpMode=RCIM_Cubic, leaveTangent=-0.466, tangentMode=RCTM_User, time=0, value=0.5}, <at 1 0.2>]}}})
    Index 0 Param 2=&FloatFromCurve(FloatCurve=DataInterface() {propertyValues {Curve {keys [{arriveTangent=-0.1866, interpMode=RCIM_Cubic, leaveTangent=-0.1866, tangentMode=RCTM_User, time=0, value=1}, <at 1 0.9>]}}})
    Index 0 Param 3=&FloatFromCurve(FloatCurve=DataInterface() {propertyValues {Curve {keys [<at 0 0>,<at 0.15 1>,<at 1 1.2>]}}})
    Module.Param0WriteEnabled=true
    Module.Param1WriteEnabled=true
    Module.Param2WriteEnabled=true
    Module.Param3WriteEnabled=false
   )
  ]
  rendererClasses [&NiagaraMeshRendererProperties]
  renderers [{rendererClass=&NiagaraMeshRendererProperties, rendererIndex=0}]
 }
 Burst_Ring001 {rendererClasses [], renderers []}
 Minimal {
  bEnabled=false
  emitterUpdateScript [
   &EmitterState()
   &SpawnBurst_Instantaneous(Use Spawn Probability=false, UseLoopCountLimit=false)
  ]
  particleSpawnScript [
   &InitializeParticle(
    Color=<rgba 0 2.4254 5 1>
    Lifetime=0.4
    Position Offset=<0 0 10>
    Sprite Size Mode=::NewEnumerator3
    Uniform Sprite Size=120
    UsePositionOffset=true
    Write Material Random=true
   )
   &SpriteFacingAndAlignment(Sprite Facing=<0 0 1>, WriteAlignment=true, WriteFacing=true)
  ]
  particleUpdateScript [
   &ParticleState()
   &ScaleSpriteSize(
    Uniform Curve Scale=2
    Uniform Curve Sprite Scale=DataInterface() {propertyValues {Curve {keys [{tangentMode=RCTM_User, time=0, value=0.2}, {tangentMode=RCTM_User, time=1, value=1}]}}}
   )
   &SpriteRotationRate(Rotation Rate=600)
   &ScaleColor(
    Scale Alpha=&FloatFromCurve(FloatCurve=DataInterface() {propertyValues {Curve {keys [<at 0 1>,<at 0.75 0.05>,<at 1 0>]}}})
    ScaleA=true
    ScaleRGB=true
   )
  ]
  rendererClasses [&NiagaraSpriteRendererProperties]
  renderers [{rendererClass=&NiagaraSpriteRendererProperties}]
 }
 Minimal001 {rendererClasses [], renderers []}
 TestEQ {
  bEnabled=false
  emitterUpdateScript [
   &EmitterState(
    Inactive Response=::NewEnumerator1
    Life Cycle Mode=::NewEnumerator1
    Loop Behavior=::NewEnumerator1
    Loop Duration Mode=::NewEnumerator1
    UseLoopDelay=true
   )
   &SpawnBurst_Instantaneous(Spawn Count=15, Spawn Group=1, Use Spawn Probability=false, UseLoopCountLimit=false) {module=SpawnBurst_Instantaneous002}
   &SpawnBurst_Instantaneous(Spawn Count=15, Use Spawn Probability=false, UseLoopCountLimit=false) {module=SpawnBurst_Instantaneous003}
  ]
  particleSpawnScript [
   &FN_SetSpawnGroup(Number of Spawn Groups=2)
   &InitializeParticle(
    Color Maximum=<rgba 2 2 2 1>
    Color Minimum=<rgba 1 1 1 1>
    Color Mode=::NewEnumerator2
    Lifetime=0.35
    Mesh Scale Max=<1 1 15>
    Mesh Scale Min=<1 1 10>
    Mesh Scale Mode=::NewEnumerator2
    Position Offset=<0 0 -5>
    Sprite Rotation Mode=::NewEnumerator2
    Sprite Size Mode=::NewEnumerator1
    Uniform Sprite Size Max=15
    Uniform Sprite Size Min=10
    UsePositionOffset=true
    Write Material Random=true
   )
   &InitialMeshOrientation(
    Mesh Orientation Mode=::NewEnumerator2
    Rotation=&RandomRangeVector(Minimum=<0 0 0>, Maximum=<0 0 1>)
    Use Rotation Vector=true
   )
   &ShapeLocation(
    Additional Quaternion Rotation=true
    Additional Yaw / Pitch / Roll=<0 0 0.5>
    MaskSpawnGroup=false
    Offset=<0 0 10>
    Offset Mode=::NewEnumerator0
    Override Seed=false
    Ring Radius=50
    Rotation Mode=::NewEnumerator3
    Shape Primitive=::NewEnumerator4
    UseOwnerScale=true
   )
   &CameraOffset(Camera Offset Amount=20)
  ]
  particleUpdateScript [
   &ParticleState()
   &ScaleMeshSize(Scale Factor=&Lerp_Vector(
     A=<1 1 0>
     B=<1 1 1>
     Alpha=&FloatFromCurve(FloatCurve=DataInterface() {propertyValues {Curve {defaultValue=0, keys [{tangentMode=RCTM_User, time=0, value=0}, {tangentMode=RCTM_User, time=1, value=1}]}}})
    ))
   &ScaleColor(
    Scale Alpha=&FloatFromCurve(
     FloatCurve=DataInterface() {propertyValues {Curve {keys [
         <at 0 0 RCIM_Cubic>
         {arriveTangent=0.037, interpMode=RCIM_Cubic, leaveTangent=0.037, tangentMode=RCTM_User, time=0.3, value=1}
         {arriveTangent=-1.7625, interpMode=RCIM_Cubic, leaveTangent=-1.7625, tangentMode=RCTM_User, time=0.6, value=0.4}
         <at 1 0 RCIM_Cubic>
        ]}}}
     Scale Curve=&Multiply_Float(
      A=&Multiply_Float(A=5, B=&FNTimeOfDayScalar(Night Scale=0.8, Morning Scale=1, Evening Scale=1))
      B=&Multiply_Float(A=&FN_FrontEndFloatModifier(Gameplay=1, Front End=0.5), B=0.25)
     )
    )
    Scale RGB=&MakeVectorFromLinearColorRGB(Color=&ColorFromCurve(ColorCurve=DataInterface() {propertyValues {
        AlphaCurve {keys []}
        BlueCurve {keys [<at 0.1903 1>,<at 0.4323 1>,<at 0.5242 1>,<at 0.8508 1 RCIM_Cubic>]}
        GreenCurve {keys [<at 0.1903 0.7555>,<at 0.4323 0.4405>,<at 0.5242 0.4205>,<at 0.8508 0.8946 RCIM_Cubic>]}
        RedCurve {keys [<at 0.1903 0.3222>,<at 0.4323 0.2991>,<at 0.5242 0.3175>,<at 0.8508 0.2222 RCIM_Cubic>]}
       }}))
    ScaleA=true
    ScaleRGB=true
   )
   &SolveForcesAndVelocity(Clamp Velocity=false, Limit Acceleration=false)
  ]
  rendererClasses [&NiagaraMeshRendererProperties]
  renderers [{rendererClass=&NiagaraMeshRendererProperties, rendererIndex=0}]
 }
 TestEQ001 {rendererClasses [], renderers []}
}
summary {
 systemName=NS_Emote_FlossYawn_Burstlight
 emitters [
  {bEnabled=false, emitterName=Burst_Ring, rendererClasses [&NiagaraMeshRendererProperties]}
  {bEnabled=false, emitterName=TestEQ, rendererClasses [&NiagaraMeshRendererProperties]}
  {bEnabled=false, emitterName=Minimal, rendererClasses [&NiagaraSpriteRendererProperties]}
  {emitterName=Burst_Ring001, rendererClasses []}
  {emitterName=Minimal001, rendererClasses []}
  {emitterName=TestEQ001, rendererClasses []}
 ]
 userVariables [
  User.bIsFrontEnd {type=NiagaraBool}
  User.bIsFrontEndPreview {type=NiagaraBool}
 ]
}
