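# Niagara system for the FlossYawn player emote that emits icy blue snowflakes with glows, flickering streaks and sparkles.
# Snowflakes spawn within a sphere, fall with curl-noise drift and write to a death position array.
# A system ReadDeath module pairs with that array, and Death_Sparkle spawns flipbook sprites at the system's death position.
# Sparkles samples particles from an emitter named "Minimal", which is absent from the listed emitters.
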NiagaraSystem /BRCosmetics/Animation/Game/MainPlayer/Emotes/FlossYawn/FX/FX/NS_Emote_FlossYawn_Emission
dependencies {
 usedDataInterfaces [
  &NiagaraDataInterfaceCurve
  &NiagaraDataInterfaceColorCurve
  &NiagaraDataInterfaceSpriteRendererInfo
  &NiagaraDataInterfaceParticleRead
  &NiagaraDataInterfaceArrayPosition
 ]
 usedDynamicInputs []
 usedModules [
  &self:SystemSpawnScript.SystemScriptSource.SystemScriptGraph.NiagaraNodeAssignment_1.SetVariables_#1
  &SystemState
  &self:ReadDeath
  &self:SystemSpawnScript.SystemScriptSource.SystemScriptGraph.NiagaraNodeAssignment_0.SetVariables_#2
 ]
}
emitters {
 Death_Sparkle {
  emitterUpdateScript [
   &EmitterState()
   &SpawnPerFrame(Spawn:NiagaraBool=$System.DeathSpawn, Use Spawn Probability=false)
  ]
  particleSpawnScript [
   &InitializeParticle(
    Color=<rgba 0.1234 0.6974 1 1>
    Lifetime Max=0.3
    Lifetime Min=0.2
    Lifetime Mode=::NewEnumerator1
    Position=$System.DeathPosition
    Position Mode=::NewEnumerator1
    Sprite Rotation Angle=&RandomRangeFloat(Minimum=-0.15, Maximum=0.15)
    Sprite Rotation Mode=::NewEnumerator1
    Sprite Size Mode=::NewEnumerator3
    Uniform Sprite Size=20
    UsePositionOffset=false
    Write Material Random=true
   )
   &CameraOffset()
  ]
  particleUpdateScript [
   &ParticleState()
   &SpriteRotationRate(Rotation Rate=&Multiply_Float(
     A=&RandomRangeFloat(Minimum=-10, Maximum=10)
     B=&FloatFromCurve(FloatCurve=DataInterface() {propertyValues {Curve {keys [<at 0 1>,<at 1 0>]}}}, Scale Curve=10)
    ))
   &ScaleSpriteSize(
    Uniform Curve Scale=&RandomRangeFloat(Minimum=0.6, Maximum=1)
    Uniform Curve Sprite Scale=DataInterface() {propertyValues {Curve {keys [{tangentMode=RCTM_User, time=0, value=0.4}, <at 0.1 1.2>, <at 0.2 1>, {tangentMode=RCTM_User, time=1, value=1}]}}}
   )
   /SetVariables(Particles.Size002:Vector2f=&Multiply_Vector2D(
     A=$Particles.SpriteSize
     B=&Vector2DFromFloat(Value=&RandomRangeFloat(Minimum=0.5, Maximum=0.8))
    )) {isSetParameters=true, module=SetVariables_#3}
   &ScaleColor(
    Scale Alpha=&FloatFromCurve(FloatCurve=DataInterface() {propertyValues {Curve {keys [<at -0.0046 1.0113 RCIM_Cubic>,<at 1.0011 0.002 RCIM_Cubic>]}}})
    ScaleA=true
    ScaleRGB=true
   )
   /SetVariables(Particles.Size003:Vector2f=&Multiply_Vector2D(A=<10 10>, B=&Vector2DFromFloat(Value=1))) {isSetParameters=true, module=SetVariables_#4}
   /SetVariables(Particles.Color003:LinearColor=&MakeLinearColorFromVectorAndFloat(
     "Vector (RGB)"=&MakeVectorFromLinearColorRGB(Color=<rgba 0.0781 0.4051 1 1>)
     "Float (Alpha)"=&FloatFromCurve(FloatCurve=DataInterface() {propertyValues {Curve {keys [<at 0 0 RCIM_Cubic>,<at 0.4 1 RCIM_Cubic>,<at 1 0 RCIM_Cubic>]}}})
    )) {isSetParameters=true, module=SetVariables_#5}
   &SolveForcesAndVelocity(Clamp Velocity=false, Limit Acceleration=false)
   &SubUVAnimation(
    End Frame Range Override=3
    "Sprite Renderer "=DataInterface() {propertyValues {SpriteRenderer=&self:Death_Sparkle_0.NiagaraSpriteRendererProperties_3}}
    UseEndFrame=true
    UseStartFrame=true
   )
  ]
  rendererClasses [&NiagaraSpriteRendererProperties]
  renderers [{rendererClass=&NiagaraSpriteRendererProperties}]
 }
 Snowflakes {
  emitterSpawnScript [/SetVariables(Emitter.EmissiveAdjustment=1) {isSetParameters=true}]
  emitterUpdateScript [
   &EmitterState()
   &SpawnRate(SpawnRate=5, Use Spawn Probability=false)
  ]
  particleSpawnScript [
   &InitializeParticle(
    Color=&RandomRangeLinearColor(Minimum=<rgba 0.5145 0.8502 1.1 100>, Maximum=<rgba 0.4554 0.706 1.1 150>)
    Lifetime=2.5
    Sprite Size Mode=::NewEnumerator3
    Uniform Sprite Size=/Niagara/DynamicInputs/Helpers/DistributionCurveFloat(
     Minimum:NiagaraFloat=5
     Maximum:NiagaraFloat=22
     Distribution Curve:NiagaraDataInterfaceCurve=DataInterface() {propertyValues {Curve {defaultValue=0, keys [<at 0 0 RCIM_Constant>,<at 0.15 0.5 RCIM_Constant>,<at 0.8 1 RCIM_Constant>]}}}
     Randomness Mode:ENiagaraRandomnessMode=::NewEnumerator0
     Evaluation Type:ENiagaraRandomnessEvaluation=::NewEnumerator0
     Seed:NiagaraInt32=0
     Fixed Random Seed:NiagaraBool=false
    )
    UsePositionOffset=false
    Write Material Random=true
   )
   &ShapeLocation(MaskSpawnGroup=false, Override Seed=false, Sphere Radius=85, UseOwnerScale=true)
   &CameraOffset()
  ]
  particleUpdateScript [
   &ParticleState()
   &ScaleSpriteSize(Uniform Curve Sprite Scale=DataInterface() {propertyValues {Curve {
       defaultValue=0
       keys [
        {interpMode=RCIM_Cubic, tangentMode=RCTM_User, time=0, value=0}
        {interpMode=RCIM_Cubic, tangentMode=RCTM_User, time=0.1, value=1}
        {arriveTangent=-1.1111, interpMode=RCIM_Cubic, leaveTangent=-1.1111, time=0.8895, value=0.7441}
        {tangentMode=RCTM_User, time=1, value=0}
       ]
      }}})
   &SpriteRotationRate(Rotation Rate=&RandomRangeFloat(Minimum=20, Maximum=50))
   &self:WriteDeath_01(Position Array:NiagaraDataInterfaceArrayPosition=$System.DeathArray) {module=WriteDeath_01}
   &DynamicMaterialParameters(
    Index 0 Param 1=&FloatFromCurve(FloatCurve=DataInterface() {propertyValues {Curve {keys [<at 0 1.1>,<at 0.2 0>,<at 0.85 0>,<at 1 1>]}}})
    Index 0 Param 2=98.7492
    Index 0 Param 4=1
    Module.Index0Float0Write=false
    Module.Index0Float1Write=true
    Module.Index0Float2Write=true
    Module.Index0Float3Write=true
   )
   &GravityForce(Gravity=&RandomRangeVector(Minimum=<0 0 -30>, Maximum=<0 0 -60>))
   &CurlNoiseForce(Bias Noise=false, Noise Strength=30, Pan Noise=false)
   &SolveForcesAndVelocity(Clamp Velocity=false, Limit Acceleration=false)
   /SetVariables(Particles.GlowScale:Vector2f=&Multiply_Vector2D(A=<1.1 1.1>, B=$Particles.SpriteSize)) {isSetParameters=true}
   /SetVariables(Particles.GlowColor=&ColorFromCurve(ColorCurve=DataInterface() {propertyValues {
       AlphaCurve {keys [<at 0 0 RCIM_Cubic>,<at 0.0335 0.25>,<at 1 0 RCIM_Cubic>]}
       BlueCurve {keys [<at 0 1 RCIM_Cubic>,<at 1 1 RCIM_Cubic>]}
       GreenCurve {keys [<at 0 0.4764 RCIM_Cubic>,<at 1 0.1837 RCIM_Cubic>]}
       RedCurve {keys [<at 0 0.0859 RCIM_Cubic>,<at 1 0 RCIM_Cubic>]}
      }})) {isSetParameters=true}
   /SetVariables(Particles.Line Color:LinearColor=&MultiplyLinearColor(
     Linear Color=<rgba 3.8016 8.944 10 1>
     Scale Factor=&MakeVector4(
      X=1
      Y=1
      Z=1.5
      W=&Multiply_Float(
       A=&Lerp_Float(
        A=0.005
        B=0.01
        Alpha=&FloatFromCurve(FloatCurve=DataInterface() {propertyValues {Curve {keys [<at 0 0>,<at 0.096 0.8735>,<at 0.1302 0.0844>,<at 0.2172 0.8972>,<at 0.2722 -0.0207>,<at 0.3649 0.909>,<at 0.4194 0.0849>,<at 0.5166 0.9082>,<at 0.5313 0.0253>,<at 0.6308 0.9318>,<at 0.7036 0.006>,<at 0.7872 0.8487>,<at 0.849 0.0772>,<at 0.9054 0.8724>,<at 1 1>]}}})
       )
       B=&RandomRangeFloat(Minimum=0.05, Maximum=0.1)
      )
     )
    )) {isSetParameters=true}
   /SetVariables(Particles.LineScale:Vector2f=&Lerp_Vector2(
     A=<200 2>
     B=<50 2>
     Alpha=&FloatFromCurve(FloatCurve=DataInterface() {propertyValues {Curve {keys [<at 0 0>,<at 0.096 0.8735>,<at 0.1302 0.0844>,<at 0.2172 0.8972>,<at 0.2722 -0.0207>,<at 0.3649 0.909>,<at 0.4194 0.0849>,<at 0.5166 0.9082>,<at 0.5313 0.0253>,<at 0.6308 0.9318>,<at 0.7036 0.006>,<at 0.7872 0.8487>,<at 0.849 0.0772>,<at 0.9054 0.8724>,<at 1 1>]}}})
    )) {isSetParameters=true}
   /SetVariables() {isSetParameters=true}
   &ScaleColor(
    Scale Alpha=&FloatFromCurve(FloatCurve=DataInterface() {propertyValues {Curve {keys [<at 0 1>,<at 0.1729 0.9441>,<at 0.7366 0.9676>,<at 0.7629 0.1912>,<at 0.7876 0.9912>,<at 0.8122 0.1206>,<at 0.8566 1.0029>,<at 1 1>]}}})
    ScaleA=true
    ScaleRGB=true
   )
  ]
  rendererClasses [&NiagaraSpriteRendererProperties]
  renderers [
   {rendererClass=&NiagaraSpriteRendererProperties}
   {rendererClass=&NiagaraSpriteRendererProperties, rendererIndex=1}
   {rendererClass=&NiagaraSpriteRendererProperties, rendererIndex=2}
  ]
 }
 Sparkles {
  emitterUpdateScript [
   &EmitterState()
   &SpawnParticlesFromOtherEmitter(
    Attribute Reader=DataInterface() {propertyValues {EmitterBinding {bindingMode=Other, emitterName=Minimal}}}
    Spawn Rate=5
    Use Maximum Age=false
    Use Minimum Age=false
   )
  ]
  particleSpawnScript [
   &InitializeParticle(
    Color=&RandomRangeLinearColor(Minimum=<rgba 0.4427 0.8727 1 0.1>, Maximum=<rgba 0.4427 0.8727 1 1>)
    Lifetime Max=0.3
    Lifetime Min=0.2
    Lifetime Mode=::NewEnumerator1
    Position Mode=::NewEnumerator1
    Sprite Rotation Angle=&RandomRangeFloat(Minimum=0, Maximum=0)
    Sprite Rotation Mode=::NewEnumerator1
    Sprite Size Mode=::NewEnumerator1
    Uniform Sprite Size Max=20
    Uniform Sprite Size Min=10
    UsePositionOffset=false
    Write Material Random=true
   )
   &CameraOffset()
   &SampleParticlesFromOtherEmitter()
   &ShapeLocation(MaskSpawnGroup=false, Override Seed=false, Sphere Radius=5, UseOwnerScale=true)
  ]
  particleUpdateScript [
   &ParticleState()
   &SpriteRotationRate(Rotation Rate=&Multiply_Float(
     A=&RandomRangeFloat(Minimum=-8, Maximum=8)
     B=&FloatFromCurve(FloatCurve=DataInterface() {propertyValues {Curve {keys [<at 0 1>,<at 1 0>]}}}, Scale Curve=10)
    ))
   &ScaleSpriteSize(
    Uniform Curve Scale=&RandomRangeFloat(Minimum=0.6, Maximum=1)
    Uniform Curve Sprite Scale=DataInterface() {propertyValues {Curve {keys [{tangentMode=RCTM_User, time=0, value=0.4}, <at 0.1 1.2>, <at 0.2 1>, {tangentMode=RCTM_User, time=1, value=1}]}}}
   )
   /SetVariables(Particles.Size002:Vector2f=&Multiply_Vector2D(
     A=$Particles.SpriteSize
     B=&Vector2DFromFloat(Value=&RandomRangeFloat(Minimum=0.5, Maximum=0.8))
    )) {isSetParameters=true, module=SetVariables_#3}
   &ScaleColor(
    Scale Alpha=&FloatFromCurve(FloatCurve=DataInterface() {propertyValues {Curve {keys [<at -0.0046 1.0113 RCIM_Cubic>,<at 1.0011 0.002 RCIM_Cubic>]}}})
    ScaleA=true
    ScaleRGB=true
   )
   /SetVariables(Particles.Size003:Vector2f=&Multiply_Vector2D(A=<10 10>, B=&Vector2DFromFloat(Value=1))) {isSetParameters=true, module=SetVariables_#4}
   /SetVariables(Particles.Color003:LinearColor=&MakeLinearColorFromVectorAndFloat(
     "Vector (RGB)"=&MakeVectorFromLinearColorRGB(Color=<rgba 0.0781 0.4051 1 1>)
     "Float (Alpha)"=&FloatFromCurve(FloatCurve=DataInterface() {propertyValues {Curve {keys [<at 0 0 RCIM_Cubic>,<at 0.4 1 RCIM_Cubic>,<at 1 0 RCIM_Cubic>]}}})
    )) {isSetParameters=true, module=SetVariables_#5}
   &SolveForcesAndVelocity(Clamp Velocity=false, Limit Acceleration=false)
   &SubUVAnimation(
    End Frame Range Override=3
    "Sprite Renderer "=DataInterface() {propertyValues {SpriteRenderer=&self:Sparkles_0.NiagaraSpriteRendererProperties_3}}
    UseEndFrame=true
    UseStartFrame=true
   )
  ]
  rendererClasses [&NiagaraSpriteRendererProperties]
  renderers [
   {rendererClass=&NiagaraSpriteRendererProperties}
   {rendererClass=&NiagaraSpriteRendererProperties, rendererIndex=1}
  ]
 }
}
summary {
 systemName=NS_Emote_FlossYawn_Emission
 emitters [
  {emitterName=Snowflakes, rendererClasses [&NiagaraSpriteRendererProperties]}
  {emitterName=Death_Sparkle, rendererClasses [&NiagaraSpriteRendererProperties]}
  {emitterName=Sparkles, rendererClasses [&NiagaraSpriteRendererProperties]}
 ]
 userVariables [
  User.bIsFrontEnd {type=NiagaraBool}
  User.bIsFrontEndPreview {type=NiagaraBool}
  User.bisHeadHidden {type=NiagaraBool}
  User.MusicReactive {type=NiagaraBool, defaultValue=true}
 ]
}
