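# Niagara system for the frost effect of the FlossYawn player emote, combining icy light, mist, particles, ground decals and sparkles.
# Secondaries merges five spawn groups into one emitter, using SpawnGroup switches and visibility tags across sprite and mesh renderers.
# Snow_Trail spawns decals at rates set by owner velocity, and Sparkles spawns from those particles through an attribute reader.
# A cyan light fades in, while time-of-day and front-end inputs adjust particle color and behavior.
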NiagaraSystem /BRCosmetics/Animation/Game/MainPlayer/Emotes/FlossYawn/FX/FX/NS_Emote_FlossYawn_Frost
dependencies {
 usedDataInterfaces [
  &NiagaraDataInterfaceArrayColor
  &NiagaraDataInterfaceArrayInt32
  &NiagaraDataInterfaceCurve
  &NiagaraDataInterfaceArrayFloat
  &NiagaraDataInterfaceSpriteRendererInfo
  &NiagaraDataInterfaceArrayFloat3
  &NiagaraDataInterfaceColorCurve
  &NiagaraDataInterfaceParticleRead
  &NiagaraDataInterfaceArrayPosition
 ]
 usedDynamicInputs [&Multiply_VectorByFloat]
 usedModules [
  &self:SystemSpawnScript.SystemScriptSource.SystemScriptGraph.NiagaraNodeAssignment_1.SetVariables_#1
  &SystemState
  &BlendSpaceHelper
  &self:ReadDeath
  &self:SystemSpawnScript.SystemScriptSource.SystemScriptGraph.NiagaraNodeAssignment_2.SetVariables_#2
 ]
}
emitters {
 Light {
  emitterUpdateScript [
   &EmitterState(Life Cycle Mode=::NewEnumerator1, Loop Behavior=::NewEnumerator1, Loop Duration Mode=::NewEnumerator1, UseLoopDelay=false)
   &SpawnBurst_Instantaneous(Use Spawn Probability=false, UseLoopCountLimit=false)
  ]
  particleSpawnScript [&InitializeParticle(Color Mode=::NewEnumerator0, Lifetime=5, UsePositionOffset=true, Write Material Random=true)]
  particleUpdateScript [
   &ParticleState(Kill Particles When Lifetime Has Elapsed=false, Let Infinitely Lived Particles Die When Emitter Deactivates=true)
   &SystemLocation(MaskSpawnGroup=false, Offset=<0 0 40>, Offset Coordinate Space=::Simulation)
   &FN_Light_Attributes(
    Color=<rgba 0.5433 5.1515 8 1>
    Create Color Attribute=true
    Create Exponent Attribute=true
    Create Position Attribute=true
    Create Radius Attribute=true
    Exponent=15
    Radius=200
    Scale Color=&FloatFromCurve(FloatCurve=DataInterface() {propertyValues {Curve {keys [<at 0 0>,<at 0.15 1>,<at 0.8 1>]}}})
   )
   /SetVariables(Particles.UpLIghtPos:NiagaraPosition=&AddVectorToPosition(Position=<0 0 0>, Vector=<0 0 250>)) {isSetParameters=true}
  ]
  rendererClasses [&NiagaraLightRendererProperties]
  renderers [{rendererClass=&NiagaraLightRendererProperties, rendererIndex=0}]
 }
 Secondaries {
  emitterUpdateScript [
   &EmitterState(Life Cycle Mode=::NewEnumerator1, Loop Duration=1, UseLoopDelay=false)
   &SpawnRate(SpawnRate=30, Use Spawn Probability=false) {module=SpawnRate001}
   &SpawnRate(SpawnGroup=1, SpawnRate=30, Use Spawn Probability=false)
   &SpawnPerUnit(Module.Use Max Movement Threshold=true, Module.Use Movement Tolerance=false, Spawn Group=2, Spawn Spacing=6, Use Spawn Probability=false)
   &SpawnRate(SpawnGroup=2, SpawnRate=6, Use Spawn Probability=false) {module=SpawnRate002}
   &SpawnRate(SpawnGroup=3, SpawnRate=150, Use Spawn Probability=false) {module=SpawnRate003}
   &SpawnRate(SpawnGroup=4, SpawnRate=3, Use Spawn Probability=false) {module=SpawnRate004}
  ]
  particleSpawnScript [
   &InitializeParticle(
    Color=&MergeHelper_ColorSwitch(
     0=&MultiplyLinearColor(
      Linear Color=&FNTimeOfDayColor(
       Night Color=<rgba 0.539 0.6881 1 0.35>
       Morning Color=<rgba 0.5751 0.8119 1 0.2>
       Day Color=<rgba 0.7059 0.9127 1 0.2>
       Evening Color=<rgba 0.5751 0.8119 1 0.2>
      )
      Scale Factor=<1 1 1 0.5>
     )
     1=&MultiplyLinearColor(
      Linear Color=&FNTimeOfDayColor(
       Night Color=<rgba 0.539 0.6881 1 0.35>
       Morning Color=<rgba 0.5751 0.8119 1 0.2>
       Day Color=<rgba 0.7059 0.9127 1 0.2>
       Evening Color=<rgba 0.5751 0.8119 1 0.2>
      )
      Scale Factor=<1 1 1 0.5>
     )
     2=&MultiplyLinearColor(
      Linear Color=&RandomRangeLinearColor(Minimum=<rgba 5 5 5 8>, Maximum=<rgba 1 1 1 1>)
      Scale Factor=&MakeVector4_FromLinearColor(LinearColor=&FNTimeOfDayColor(Night Color=<rgba 1 1 1 1>, Morning Color=<rgba 1 1 1 1>, Day Color=<rgba 1 1 1 1.2>, Evening Color=<rgba 1 1 1 1>))
     )
     3=<rgba 1 1 1 1>
     4=&RandomRangeLinearColor(Minimum=<rgba 0 1.4743 1.5 1>, Maximum=<rgba 0 0.1344 1 1>)
     Index=&SpawnGroup()
     Num=::NewEnumerator3
    )
    Lifetime=&MergeHelper_FloatSwitch(
     0=&RandomRangeFloat(Minimum=0.7, Maximum=1.4)
     1=&RandomRangeFloat(Minimum=0.7, Maximum=1.4)
     2=&RandomRangeFloat(Minimum=2, Maximum=2.5)
     3=&RandomRangeFloat(Minimum=1, Maximum=2)
     4=&RandomRangeFloat(Minimum=0.8, Maximum=3)
     Index=&SpawnGroup()
     Num=::NewEnumerator3
    )
    Mesh Scale Max=<3 2 15>
    Mesh Scale Min=<2 2 10>
    Mesh Scale Mode=::NewEnumerator2
    Position Offset=&MergeHelper_VectorSwitch(0=<0 0 0>, 1=<0 0 0>, 2=<0 0 -10>, 3=<0 0 150>, 4=<0 0 0>, Index=&SpawnGroup(), Num=::NewEnumerator3)
    Sprite Rotation Mode=::NewEnumerator2
    Sprite Size Mode=::NewEnumerator3
    Uniform Sprite Size=&MergeHelper_FloatSwitch(
     0=&RandomRangeFloat(Minimum=15, Maximum=35)
     1=&RandomRangeFloat(Minimum=15, Maximum=35)
     2=0
     3=&RandomRangeFloat(Minimum=5, Maximum=10)
     4=&RandomRangeFloat(Minimum=6, Maximum=24)
     Index=&SpawnGroup()
     Num=::NewEnumerator3
    )
    UsePositionOffset=true
    Write Material Random=true
   )
   &InitialMeshOrientation(
    Mesh Orientation Mode=::NewEnumerator2
    Rotation=&RandomRangeVector(Minimum=<0 0 0>, Maximum=<0 0 1>)
    Use Rotation Vector=true
   )
   &ShapeLocation(
    Additional Quaternion Rotation=true
    Additional Yaw / Pitch / Roll=<0 0 0.5>
    Apply To Particle Position=::NewEnumerator1
    MaskSpawnGroup=false
    Offset=<0 0 10>
    Offset Mode=::NewEnumerator0
    Override Seed=false
    Ring Radius=60
    Rotation Mode=::NewEnumerator3
    Shape Primitive=::NewEnumerator4
    U Distribution=0.5
    UseOwnerScale=true
   )
   &ShapeLocation(
    Additional Quaternion Rotation=false
    Apply To Particle Position=::NewEnumerator1
    Cylinder Height=2
    Cylinder Radius=15
    MaskSpawnGroup=true
    Non Uniform Scale=<1 1.5 1>
    Offset=&SelectVectorFromArray(
     Array Sampling Mode=::NewEnumerator0
     Direct Array Index=?
     Vector Selection Array=DataInterface() {propertyValues {FloatData [<50 0 0>, <-50 0 0>]}}
     Random Seed=0
    )
    Offset Coordinate Space=::Local
    Offset Mode=::NewEnumerator0
    Override Seed=false
    Rotation Mode=::NewEnumerator3
    Shape Primitive=::NewEnumerator1
    Surface Only=false
    UseOwnerScale=true
   ) {module=ShapeLocation001}
   &ShapeLocation(
    Apply To Particle Position=::NewEnumerator1
    Box Size=<150 150 250>
    MaskSpawnGroup=false
    Offset=<0 0 -10>
    Offset Mode=::NewEnumerator0
    Override Seed=false
    Shape Primitive=::NewEnumerator2
    Surface Only=false
    UseOwnerScale=true
   ) {module=ShapeLocation002}
   &ShapeLocation(
    Apply To Particle Position=::NewEnumerator1
    MaskSpawnGroup=false
    Offset=<0 0 150>
    Offset Mode=::NewEnumerator0
    Override Seed=false
    UseOwnerScale=true
   ) {module=ShapeLocation003}
   /SetVariables(Particles.Position=&MergeHelper_VectorSwitch(
     0=Linked() {linkedVariable {name=Particles.ShapeLocation.ShapePosition, type=NiagaraPosition}}
     1=Linked() {linkedVariable {name=Particles.ShapeLocation.ShapePosition, type=NiagaraPosition}}
     2=Linked() {linkedVariable {name=Particles.ShapeLocation001.ShapePosition, type=NiagaraPosition}}
     3=Linked() {linkedVariable {name=Particles.ShapeLocation002.ShapePosition, type=NiagaraPosition}}
     4=Linked() {linkedVariable {name=Particles.ShapeLocation003.ShapePosition, type=NiagaraPosition}}
     Index=&SpawnGroup()
     Num=::NewEnumerator3
    )) {isSetParameters=true}
   &AddVelocity(
    Additional Quaternion Rotation=true
    Additional Yaw / Pitch / Roll:Vector3f=<0 0 180>
    Rotation Mode=::NewEnumerator3
    Rotation Quaternion:Quat4f=$Engine.Owner.Rotation
    Velocity=<0 100 -10>
   )
   /SetVariables(Particles.Velocity=&MergeHelper_VectorSwitch(0=$Particles.Velocity, 1=<0 0 0>, 2=<0 0 0>, 3=<0 0 0>, 4=<0 0 0>, Index=&SpawnGroup(), Num=::NewEnumerator3)) {isSetParameters=true}
   /SetVariables(Particles.VisibilityTag=&MergeHelper_IntSwitch(0=0, 1=0, 2=1, 3=2, 4=3, 5=?, Index=&SpawnGroup(), Num=::NewEnumerator3)) {isSetParameters=true}
   &CameraOffset(Camera Offset Amount=&MergeHelper_FloatSwitch(0=0, 1=0, 2=20, 3=0, 4=0, Index=&SpawnGroup(), Num=::NewEnumerator3))
  ]
  particleUpdateScript [
   &ParticleState()
   /Niagara/Modules/Update/SubUV/SubUVAnimation(
    Number Of Frames=&Add_Float(
     A=&Lerp_Float(A=0, B=30, Alpha=$Particles.NormalizedAge)
     B=&RandomRangeFloat(Minimum=0, Maximum=30)
    )
    SubUV Lookup Index=&Multiply_Float(A=$Particles.NormalizedAge, B=1)
   )
   &ScaleSpriteSize(Scale Factor=&MergeHelper_Vector2Switch(
     0=&Vector2DFromFloat(Value=&FloatFromCurve(
       FloatCurve=DataInterface() {propertyValues {Curve {keys [
           {arriveTangent=3.9723, interpMode=RCIM_Cubic, leaveTangent=3.9723, tangentMode=RCTM_User, time=0, value=0.2}
           {arriveTangent=0.8, interpMode=RCIM_Cubic, leaveTangent=0.8, time=0.209, value=0.58}
           <at 1 1 RCIM_Cubic>
          ]}}}
       Scale Curve=3
      ))
     1=&Vector2DFromFloat(Value=&FloatFromCurve(
       FloatCurve=DataInterface() {propertyValues {Curve {keys [
           {arriveTangent=3.9723, interpMode=RCIM_Cubic, leaveTangent=3.9723, tangentMode=RCTM_User, time=0, value=0.2}
           {arriveTangent=0.8, interpMode=RCIM_Cubic, leaveTangent=0.8, time=0.209, value=0.58}
           <at 1 1 RCIM_Cubic>
          ]}}}
       Scale Curve=3
      ))
     2=<0 0>
     3=&Vector2DFromFloat(Value=&FloatFromCurve(FloatCurve=DataInterface() {propertyValues {Curve {
          defaultValue=0
          keys [
           {interpMode=RCIM_Cubic, tangentMode=RCTM_User, time=0.7839, value=1}
           <at 0.8277 0 RCIM_Cubic>
           <at 0.8756 1 RCIM_Cubic>
           <at 0.9204 0 RCIM_Cubic>
           <at 0.955 1 RCIM_Cubic>
           {arriveTangent=-9.8207, interpMode=RCIM_Cubic, leaveTangent=-9.8207, tangentMode=RCTM_User, time=1, value=0}
          ]
         }}}))
     4=<1 1>
     Index=&SpawnGroup()
     Num=::NewEnumerator3
    ))
   &GravityForce(Gravity=&MergeHelper_VectorSwitch(
     0=<0 0 0>
     1=<0 0 0>
     2=<0 0 0>
     3=<0 0 -200>
     4=&RandomRangeVector(Minimum=<0 0 -10>, Maximum=<0 0 -20>)
     Index=&SpawnGroup()
     Num=::NewEnumerator3
    ))
   &Drag(
    Drag=&MergeHelper_FloatSwitch(
     0=0
     1=0
     2=0
     3=&Multiply_Float(
      A=&RandomRangeFloat(Minimum=1, Maximum=5)
      B=&FloatFromCurve(FloatCurve=DataInterface() {propertyValues {Curve {defaultValue=1, keys [{tangentMode=RCTM_User, time=0.2, value=2}, {tangentMode=RCTM_User, time=0.65, value=0.5}]}}})
     )
     4=0
     Index=&SpawnGroup()
     Num=::NewEnumerator3
    )
    Rotational Drag:NiagaraFloat=&MergeHelper_FloatSwitch(0=0, 1=0, 2=0, 3=1, 4=0, Index=&SpawnGroup(), Num=::NewEnumerator3)
    Use Linear Drag=true
    Use Rotational Drag=true
   )
   &CurlNoiseForce(
    Bias Noise=false
    Noise Frequency=&MergeHelper_FloatSwitch(0=0, 1=0, 2=0, 3=200, 4=50, Index=&SpawnGroup(), Num=::NewEnumerator3)
    Noise Strength=&MergeHelper_FloatSwitch(
     0=0
     1=0
     2=0
     3=&RandomRangeFloat(Minimum=50, Maximum=100)
     4=10
     Index=&SpawnGroup()
     Num=::NewEnumerator3
    )
    Pan Noise=false
   ) {enabled=false}
   /SetVariables(Particles.BrightColor:LinearColor=&MultiplyLinearColor(
     Linear Color=&MultiplyLinearColor(
      Linear Color=&ColorFromCurve(ColorCurve=DataInterface() {propertyValues {
         AlphaCurve {keys [<at 0 0 RCIM_Cubic>,<at 0.1239 1>,<at 1 0 RCIM_Cubic>]}
         BlueCurve {keys [<at 0 15 RCIM_Cubic>,<at 1 1 RCIM_Cubic>]}
         GreenCurve {keys [<at 0 10.4493 RCIM_Cubic>,<at 1 0.1803 RCIM_Cubic>]}
         RedCurve {keys [<at 0 2.5491 RCIM_Cubic>,<at 1 0.0307 RCIM_Cubic>]}
        }})
      Scale Factor=&MakeVector4_FromLinearColor(LinearColor=&FNTimeOfDayColor(Night Color=<rgba 1 1 1 1>, Morning Color=<rgba 1 1 1 1>, Day Color=<rgba 2 2 2 1>, Evening Color=<rgba 1 1 1 1>))
     )
     Scale Factor=&MakeVector4_FromLinearColor(LinearColor=<rgba 1.5 1.5 1.5 1>)
    )) {isSetParameters=true}
   /SetVariables(
    Particles.Bokeh Flicker:NiagaraFloat=&Waveform(
     "Waveform [1]"=::NewEnumerator0
     "[1] Frequency"=&RandomRangeFloat(Minimum=1, Maximum=5)
     "[1] Amplitude Min/Max"=<0.1 1>
    )
    Particles.Flicker:NiagaraFloat=&Waveform(
     "Waveform [1]"=::NewEnumerator0
     "[1] Frequency"=&RandomRangeFloat(Minimum=5, Maximum=10)
     "[1] Amplitude Min/Max"=<0 1>
    )
   ) {isSetParameters=true}
   /SetVariables(
    Particles.Bokeh Color:LinearColor=&MakeLinearColorFromVectorAndFloat(
     "Vector (RGB)"=&MakeVectorFromLinearColorRGB(Color=<rgba 0.3958 0.8749 1 1>)
     "Float (Alpha)"=&Multiply_Float(
      A=&FloatFromCurve(
       FloatCurve=DataInterface() {propertyValues {Curve {
          defaultValue=1
          keys [{interpMode=RCIM_Cubic, tangentMode=RCTM_User, time=0, value=1}, {interpMode=RCIM_Cubic, tangentMode=RCTM_User, time=1, value=0}]
         }}}
       Scale Curve=0.3
      )
      B=$Particles.Bokeh Flicker
     )
    )
    Particles.Bokeh Size:Vector2f=&Multiply_Vector2D_ByFloat(
     Float=&FloatFromCurve(FloatCurve=DataInterface() {propertyValues {Curve {
         defaultValue=1
         keys [
          {arriveTangent=40.8712, interpMode=RCIM_Cubic, leaveTangent=40.8712, tangentMode=RCTM_User, time=0, value=0}
          {arriveTangent=0.7672, interpMode=RCIM_Cubic, leaveTangent=0.7672, tangentMode=RCTM_User, time=0.0663, value=1.0039}
          {arriveTangent=-1.712, interpMode=RCIM_Cubic, leaveTangent=-1.712, tangentMode=RCTM_User, time=0.3106, value=0.656}
          {arriveTangent=-0.0405, interpMode=RCIM_Cubic, leaveTangent=-0.0405, tangentMode=RCTM_User, time=1, value=0.25}
         ]
        }}})
     Vector2D=&Vector2DFromFloat(Value=&RandomRangeFloat(Minimum=5, Maximum=20))
    )
   ) {isSetParameters=true}
   &ScaleColor(
    Scale Mode=::NewEnumerator1
    Scale RGBA=/Game/Effects/Niagara/Modules/MergeEmitters/MergeHelper_Vector4Switch(
     0:Vector4f=&MakeVector4_FromLinearColor(LinearColor=&ColorFromCurve(ColorCurve=DataInterface() {propertyValues {
         AlphaCurve {keys [<at 0 0 RCIM_Cubic>,<at 0.3222 1>,<at 0.4893 0.62>,<at 1 0>,<at 1 0>]}
         BlueCurve {keys [<at 0 1 RCIM_Cubic>,<at 1 1 RCIM_Cubic>]}
         GreenCurve {keys [<at 0 0.715 RCIM_Cubic>,<at 1 0.79 RCIM_Cubic>]}
         RedCurve {keys [<at 0 0.62 RCIM_Cubic>,<at 1 0.7 RCIM_Cubic>]}
        }}))
     1:Vector4f=&MakeVector4_FromLinearColor(LinearColor=&ColorFromCurve(ColorCurve=DataInterface() {propertyValues {
         AlphaCurve {keys [<at 0 0 RCIM_Cubic>,<at 0.3222 1>,<at 0.4893 0.62>,<at 1 0>,<at 1 0>]}
         BlueCurve {keys [<at 0 1 RCIM_Cubic>,<at 1 1 RCIM_Cubic>]}
         GreenCurve {keys [<at 0 0.715 RCIM_Cubic>,<at 1 0.79 RCIM_Cubic>]}
         RedCurve {keys [<at 0 0.62 RCIM_Cubic>,<at 1 0.7 RCIM_Cubic>]}
        }}))
     2:Vector4f=&MakeVector4(
      X=&MakeFloatFromVector(
       Vector=&MakeVectorFromLinearColorRGB(Color=&ColorFromCurve(ColorCurve=DataInterface() {propertyValues {AlphaCurve {keys []}, BlueCurve {keys [<at 0.1903 1>]}, GreenCurve {keys [<at 0.1903 0.7574>]}, RedCurve {keys [<at 0.1903 0.1809>]}}}))
       Channel=::NewEnumerator0
      )
      Y=&MakeFloatFromVector(
       Vector=&MakeVectorFromLinearColorRGB(Color=&ColorFromCurve(ColorCurve=DataInterface() {propertyValues {AlphaCurve {keys []}, BlueCurve {keys [<at 0.1903 1>]}, GreenCurve {keys [<at 0.1903 0.7574>]}, RedCurve {keys [<at 0.1903 0.1809>]}}}))
       Channel=::NewEnumerator1
      )
      Z=&MakeFloatFromVector(
       Vector=&MakeVectorFromLinearColorRGB(Color=&ColorFromCurve(ColorCurve=DataInterface() {propertyValues {AlphaCurve {keys []}, BlueCurve {keys [<at 0.1903 1>]}, GreenCurve {keys [<at 0.1903 0.7574>]}, RedCurve {keys [<at 0.1903 0.1809>]}}}))
       Channel=::NewEnumerator2
      )
      W=&FloatFromCurve(
       FloatCurve=DataInterface() {propertyValues {Curve {keys [
           <at 0 0 RCIM_Cubic>
           {arriveTangent=0.037, interpMode=RCIM_Cubic, leaveTangent=0.037, tangentMode=RCTM_User, time=0.3, value=1}
           {arriveTangent=-1.7625, interpMode=RCIM_Cubic, leaveTangent=-1.7625, tangentMode=RCTM_User, time=0.6, value=0.4}
           <at 1 0 RCIM_Cubic>
          ]}}}
       Scale Curve=&Multiply_Float(
        A=&Multiply_Float(A=5, B=&FNTimeOfDayScalar(Night Scale=1, Morning Scale=1, Day Scale=0.5, Evening Scale=1))
        B=&Multiply_Float(A=&FN_FrontEndFloatModifier(Gameplay=1, Front End=0.5), B=1)
       )
      )
     )
     3:Vector4f=&MakeVector4(
      X=&MakeFloatFromVector(
       Vector=&MergeHelper_VectorSwitch(
        0=<1 1 1>
        1=&MakeVectorFromLinearColorRGB(Color=&ColorFromCurve(ColorCurve=DataInterface() {propertyValues {
            AlphaCurve {keys []}
            BlueCurve {keys [<at 0.0637 3>,<at 0.1009 1>,<at 0.3329 1>,<at 0.9944 0.1323>]}
            GreenCurve {keys [<at 0.0637 3>,<at 0.1009 1>,<at 0.3329 1>,<at 0.9944 0>]}
            RedCurve {keys [<at 0.0637 3>,<at 0.1009 1>,<at 0.3329 1>,<at 0.9944 0.3646>]}
           }}))
        2=?
        3=?
        Index=&SpawnGroup()
        Num=::NewEnumerator0
       )
       Channel=::NewEnumerator0
      )
      Y=&MakeFloatFromVector(
       Vector=&MergeHelper_VectorSwitch(
        0=<1 1 1>
        1=&MakeVectorFromLinearColorRGB(Color=&ColorFromCurve(ColorCurve=DataInterface() {propertyValues {
            AlphaCurve {keys []}
            BlueCurve {keys [<at 0.0637 3>,<at 0.1009 1>,<at 0.3329 1>,<at 0.9944 0.1323>]}
            GreenCurve {keys [<at 0.0637 3>,<at 0.1009 1>,<at 0.3329 1>,<at 0.9944 0>]}
            RedCurve {keys [<at 0.0637 3>,<at 0.1009 1>,<at 0.3329 1>,<at 0.9944 0.3646>]}
           }}))
        2=?
        3=?
        Index=&SpawnGroup()
        Num=::NewEnumerator0
       )
       Channel=::NewEnumerator1
      )
      Z=&MakeFloatFromVector(
       Vector=&MergeHelper_VectorSwitch(
        0=<1 1 1>
        1=&MakeVectorFromLinearColorRGB(Color=&ColorFromCurve(ColorCurve=DataInterface() {propertyValues {
            AlphaCurve {keys []}
            BlueCurve {keys [<at 0.0637 3>,<at 0.1009 1>,<at 0.3329 1>,<at 0.9944 0.1323>]}
            GreenCurve {keys [<at 0.0637 3>,<at 0.1009 1>,<at 0.3329 1>,<at 0.9944 0>]}
            RedCurve {keys [<at 0.0637 3>,<at 0.1009 1>,<at 0.3329 1>,<at 0.9944 0.3646>]}
           }}))
        2=?
        3=?
        Index=&SpawnGroup()
        Num=::NewEnumerator0
       )
       Channel=::NewEnumerator2
      )
      W=&MergeHelper_FloatSwitch(
       0=&FloatFromCurve(FloatCurve=DataInterface() {propertyValues {Curve {
           defaultValue=1
           keys [
            {arriveTangent=24.9349, interpMode=RCIM_Cubic, leaveTangent=24.9349, tangentMode=RCTM_User, time=0, value=0}
            {arriveTangent=0.7276, interpMode=RCIM_Cubic, leaveTangent=0.7276, tangentMode=RCTM_User, time=0.1008, value=1.0033}
            {arriveTangent=-1.8102, interpMode=RCIM_Cubic, leaveTangent=-1.8102, tangentMode=RCTM_User, time=0.5915, value=0.8526}
            {arriveTangent=-0.0405, interpMode=RCIM_Cubic, leaveTangent=-0.0405, tangentMode=RCTM_User, time=1, value=0}
           ]
          }}})
       1=&Multiply_Float(
        A=&FloatFromCurve(FloatCurve=DataInterface() {propertyValues {Curve {
            defaultValue=1
            keys [
             {arriveTangent=27.2232, interpMode=RCIM_Cubic, leaveTangent=27.2232, tangentMode=RCTM_User, time=0, value=0}
             {arriveTangent=0.4162, interpMode=RCIM_Cubic, leaveTangent=0.4162, tangentMode=RCTM_User, time=0.0689, value=1.2342}
             {arriveTangent=-0.0405, interpMode=RCIM_Cubic, leaveTangent=-0.0405, tangentMode=RCTM_User, time=1, value=0}
            ]
           }}})
        B=$Particles.Flicker
       )
       2=?
       3=?
       Index=&SpawnGroup()
       Num=::NewEnumerator0
      )
     )
     4:Vector4f=&MakeVector4(
      X=1
      Y=1
      Z=1
      W=&FloatFromCurve(FloatCurve=DataInterface() {propertyValues {Curve {keys [<at 0 0>,<at 0.139 0.4>,<at 1 0>]}}})
     )
     5:Vector4f=?
     6:Vector4f=?
     7:Vector4f=?
     8:Vector4f=?
     9:Vector4f=?
     Index:NiagaraInt32=&SpawnGroup()
     Num:MergeHelper_SwitcherEnum=::NewEnumerator3
    )
    ScaleRGBA=true
   ) {module=ScaleColor001}
   &SolveForcesAndVelocity(Clamp Velocity=false, Limit Acceleration=false)
   /SetVariables(Particles.TopScale:Vector3f=&Multiply_Vector(A=$Particles.Scale, B=<1 1 0.1>)) {isSetParameters=true}
  ]
  rendererClasses [&NiagaraSpriteRendererProperties, &NiagaraMeshRendererProperties]
  renderers [
   {rendererClass=&NiagaraSpriteRendererProperties}
   {rendererClass=&NiagaraMeshRendererProperties, rendererIndex=1}
   {rendererClass=&NiagaraMeshRendererProperties, rendererIndex=2}
   {rendererClass=&NiagaraSpriteRendererProperties, rendererIndex=3}
   {rendererClass=&NiagaraSpriteRendererProperties, rendererIndex=4}
  ]
 }
 Snow_Trail {
  emitterSpawnScript [/SetVariables(Emitter.WetColor:LinearColor=&MakeCustomLinearColorFromBool(
     Bool=$User.bIsFrontEndPreview
     True Color=<rgba 0.7115 0.772 0.8229 1>
     False Color=&FNTimeOfDayColor(
      Night Color=<rgba 0.8936 0.9245 0.9375 1>
      Morning Color=<rgba 1 1 1 0.2251>
      Day Color=<rgba 0.7552 0.9175 1 0.225>
      Evening Color=<rgba 0.8 0.8968 0.9375 0.05>
     )
    )) {isSetParameters=true}]
  emitterUpdateScript [
   &EmitterState(Life Cycle Mode=::NewEnumerator1, Loop Behavior=::NewEnumerator1, Loop Duration Mode=::NewEnumerator1, UseLoopDelay=false)
   &Timeline(
    Play=&SetBoolByFloatComparison(A=&VectorLength(Vector=$User.OwnerLocalVelocity), B=60)
    Rewind Rate=5
   )
   &SpawnRate(
    SpawnGroup=1
    SpawnRate=&MakeCustomFloatFromBool(
     Bool=&SetBoolByFloatComparison(A=&VectorLength(Vector=$User.OwnerLocalVelocity), B=60)
     True Float=2
     False Float=&MakeCustomFloatFromBool(Bool=$User.bIsFrontEndPreview, True Float=4, False Float=1)
    )
    Use Spawn Probability=false
   ) {module=SpawnRate001}
  ]
  particleSpawnScript [
   /SetVariables(Particles.VisibilityTag=&SpawnGroup()) {isSetParameters=true}
   &InitializeParticle(
    Color=&MultiplyLinearColor(
     Linear Color=&SelectLinearColorFromArray(
      Array Sampling Mode=::NewEnumerator0
      Direct Array Index=?
      Color Selection Array=DataInterface() {propertyValues {ColorData [<rgba 0.4257 1 0.933 1>, <rgba 0.4592 0.9713 1 1>]}}
      Random Seed=0
     )
     Scale Factor=&MakeVector4_FromLinearColor(LinearColor=&FNTimeOfDayColor(Night Color=<rgba 1 1 1 1>, Morning Color=<rgba 1 1 1 1>, Day Color=<rgba 0.59 0.9134 1 1>, Evening Color=<rgba 1 1 1 1>))
    )
    Lifetime=2
    Position Offset=&MaskVectorBySpawnGroup(Masked Vector=<0 0 0>, Passthrough Vector=<0 0 0>)
    Position Offset Coordinate Space=::Local
    Sprite Rotation Mode=::NewEnumerator2
    Sprite Size Mode=::NewEnumerator1
    Uniform Sprite Size Max=30
    Uniform Sprite Size Min=25
    UsePositionOffset=true
    Write Material Random=true
   )
   /SetVariables(Particles.Sparkles Lifetime:NiagaraFloat=&RandomRangeFloat(Minimum=0.25, Maximum=0.25)) {isSetParameters=true}
   /SetVariables(
    Particles.DecalOrientation:Quat4f=/Game/Effects/Niagara/Modules/MergeEmitters/MergeHelper_QuatSwitch(
     0:Quat4f=<0 1 0 1>
     1:Quat4f=<0.2 1 0 1>
     2:Quat4f=<0.4 1 0 1>
     3:Quat4f=<0.6 1 0 1>
     4:Quat4f=<0.8 1 0 1>
     5:Quat4f=?
     6:Quat4f=?
     7:Quat4f=?
     8:Quat4f=?
     9:Quat4f=?
     Index:NiagaraInt32=&SelectIntFromArray(Direct Array Index=0, Int Selection Array=DataInterface() {propertyValues {IntData [0,1,2,3,4]}})
     Num:MergeHelper_SwitcherEnum=::NewEnumerator3
    )
    Particles.DecalSize:Vector3f=<200 165 165>
   ) {isSetParameters=true}
   &ShapeLocation(
    Box Size=<120 40 10>
    MaskSpawnGroup=false
    Non Uniform Scale=&MaskVector(Mask Source=0, Mask Parameter=?, Mask=1, Masked Vector=<0 0 0>, Passthrough Vector=<1 1 1>)
    Offset=<0 0 15>
    Offset Mode=::NewEnumerator0
    Override Seed=false
    Shape Primitive=::NewEnumerator2
    Surface Only=false
    UseOwnerScale=true
   )
   &CameraOffset(Camera Offset Amount=&MaskFloatBySpawnGroup(Masked Float=20, Passthrough Float=0))
   &AddVelocity(Velocity=&MakeCustomVectorFromBool(Bool=$User.bIsFrontEndPreview, True Vector=<0 -70 0>, False Vector=<0 0 0>))
   /Niagara/Modules/Spawn/Decal/InitialDecalOrientation(
    Orientation Vector:Vector3f=<0 0 1>
    Rotation:Vector3f=&RandomRangeVector(Minimum=<0 0 0>, Maximum=<0 0 1>)
    Use Rotation Vector=true
   )
  ]
  particleUpdateScript [
   &ParticleState()
   /SetVariables(Particles.Update Sparkles Age:NiagaraFloat=&Divide_Float(A=$Particles.Age, B=$Particles.Initial.Sparkles Lifetime)) {isSetParameters=true}
   /SetVariables(Particles.DecalColor:LinearColor=/Niagara/DynamicInputs/LinearColor/MakeLinearColorFromVector4(Vector:Vector4f=&MakeVector4(
      X=&FloatFromCurve(FloatCurve=DataInterface() {propertyValues {Curve {keys [<at 0 1>,<at 0.5055 0.8853>,<at 1 0>]}}}, Scale Curve=-1)
      Y=0
      Z=0
      W=&FloatFromCurve(
       FloatCurve=DataInterface() {propertyValues {Curve {keys [<at 0 0>,<at 0.1265 1>,<at 0.7005 0.8147>,<at 1 0>]}}}
       Scale Curve=&RandomRangeFloat(Minimum=3, Maximum=2)
      )
     ))) {isSetParameters=true}
   &ScaleSpriteSize(
    Uniform Curve Index=$Particles.Update Sparkles Age
    Uniform Curve Sprite Scale=DataInterface() {propertyValues {Curve {defaultValue=0, keys [{interpMode=RCIM_Cubic, tangentMode=RCTM_User, time=0, value=0}, {tangentMode=RCTM_User, time=1, value=1}]}}}
   )
   &SpriteRotationRate(Rotation Rate=&SelectFloatFromArray(
     Array Sampling Mode=::NewEnumerator0
     Direct Array Index=?
     Float Selection Array=DataInterface() {propertyValues {FloatData [0,0]}}
     Random Seed=0
    ))
   &BlendSpace("Position Blend (World <-> Local)":NiagaraFloat=&MaskFloatBySpawnGroup(
     Masked Float=&MakeCustomFloatFromBool(
      Bool=&SetBoolByFloatComparison(A=&VectorLength(Vector=$Engine.Owner.Velocity), B=400)
      True Float=1
      False Float=0
     )
     Passthrough Float=0
     Translation Contribution:NiagaraFloat=&MaskFloatBySpawnGroup(
      Masked Float=&MakeCustomFloatFromBool(
       Bool=&SetBoolByFloatComparison(A=&VectorLength(Vector=$Engine.Owner.Velocity), B=400)
       True Float=1
       False Float=0
      )
      Passthrough Float=0
     )
     Rotation Contribution:NiagaraFloat=&MaskFloatBySpawnGroup(
      Masked Float=&MakeCustomFloatFromBool(
       Bool=&SetBoolByFloatComparison(A=&VectorLength(Vector=$Engine.Owner.Velocity), B=400)
       True Float=1
       False Float=0
      )
      Passthrough Float=0
     )
    ))
   &SolveForcesAndVelocity(Clamp Velocity=false, Limit Acceleration=false)
   /SetVariables(Particles.Linear Color=&MakeLinearColorFromVector4(Vector:Vector4f=&MakeVector4(
      X=&FloatFromCurve(FloatCurve=DataInterface() {propertyValues {Curve {keys [<at 0 1>,<at 0.5055 1>,<at 1 1>]}}}, Scale Curve=-1)
      Y=0
      Z=0
      W=&FloatFromCurve(
       FloatCurve=DataInterface() {propertyValues {Curve {keys [<at 0 0>,<at 0.1265 1>,<at 0.7005 1>,<at 1 0>]}}}
       Scale Curve=&RandomRangeFloat(Minimum=2, Maximum=3)
      )
     ))) {isSetParameters=true}
   /Niagara/Modules/Update/SubUV/V2/SubUVAnimation(
    "Sprite Renderer "=DataInterface() {propertyValues {SpriteRenderer=&self:Snow_Trail_0.NiagaraSpriteRendererProperties_0}}
    SubUV Lookup Index:NiagaraFloat=$Particles.Update Sparkles Age
    UseEndFrame=false
    UseStartFrame=false
   )
   &ScaleColor(
    Scale Alpha=&FloatFromCurve(FloatCurve=DataInterface() {propertyValues {Curve {keys [<at 0 1 RCIM_Constant>,<at 0.25 0 RCIM_Constant>]}}})
    ScaleA=true
    ScaleRGB=true
   )
  ]
  rendererClasses [&NiagaraDecalRendererProperties]
  renderers [
   {rendererClass=&NiagaraDecalRendererProperties, rendererIndex=0}
   {rendererClass=&NiagaraDecalRendererProperties, rendererIndex=1}
  ]
 }
 Sparkles {
  emitterUpdateScript [
   &EmitterState()
   &SpawnParticlesFromOtherEmitter(
    Attribute Reader=DataInterface() {propertyValues {EmitterBinding {bindingMode=Other, emitterName=Snow_Trail}}}
    Spawn Rate=4
    Use Maximum Age=false
    Use Minimum Age=false
   )
  ]
  particleSpawnScript [
   &InitializeParticle(
    Color=&RandomRangeLinearColor(Minimum=<rgba 0.4427 0.8727 1 0.1>, Maximum=<rgba 0.4427 0.8727 1 1>)
    Lifetime Max=0.6
    Lifetime Min=0.4
    Lifetime Mode=::NewEnumerator1
    Position Mode=::NewEnumerator1
    Sprite Rotation Angle=&RandomRangeFloat(Minimum=0, Maximum=0)
    Sprite Rotation Mode=::NewEnumerator1
    Sprite Size Mode=::NewEnumerator1
    Uniform Sprite Size Max=20
    Uniform Sprite Size Min=10
    UsePositionOffset=false
    Write Material Random=true
   )
   &CameraOffset()
   &SampleParticlesFromOtherEmitter(
    Maximum Sampled Particle Age=1.1
    Minimum Sampled Particle Age=0.2
    Module.MaxSampledAgeCutoff=true
    Module.MinSampledAgeCutoff=true
    Reject Particle By Sampled Age=true
   )
   &ShapeLocation(
    Cylinder Height=0
    Cylinder Radius=60
    MaskSpawnGroup=false
    Offset=<0 0 -10>
    Offset Mode=::NewEnumerator0
    Override Seed=false
    Shape Primitive=::NewEnumerator1
    Surface Only=false
    UseOwnerScale=true
   ) {module=ShapeLocation001}
   &ShapeLocation(MaskSpawnGroup=false, Override Seed=false, Sphere Radius=5, UseOwnerScale=true) {enabled=false}
  ]
  particleUpdateScript [
   &ParticleState()
   &SpriteRotationRate(Rotation Rate=&Multiply_Float(
     A=&RandomRangeFloat(Minimum=-8, Maximum=8)
     B=&FloatFromCurve(FloatCurve=DataInterface() {propertyValues {Curve {keys [<at 0 1>,<at 1 0>]}}}, Scale Curve=10)
    ))
   &ScaleSpriteSize(
    Uniform Curve Scale=&RandomRangeFloat(Minimum=0.6, Maximum=1)
    Uniform Curve Sprite Scale=DataInterface() {propertyValues {Curve {keys [{tangentMode=RCTM_User, time=0, value=0.4}, <at 0.1 1.2>, <at 0.2 1>, {tangentMode=RCTM_User, time=1, value=1}]}}}
   )
   /SetVariables(Particles.Size002:Vector2f=&Multiply_Vector2D(
     A=$Particles.SpriteSize
     B=&Vector2DFromFloat(Value=&RandomRangeFloat(Minimum=0.5, Maximum=0.8))
    )) {isSetParameters=true}
   &ScaleColor(
    Scale Alpha=&FloatFromCurve(FloatCurve=DataInterface() {propertyValues {Curve {keys [<at -0.0046 1.0113 RCIM_Cubic>,<at 1.0011 0.002 RCIM_Cubic>]}}})
    ScaleA=true
    ScaleRGB=true
   )
   /SetVariables(Particles.Size003:Vector2f=&Multiply_Vector2D(A=<10 10>, B=&Vector2DFromFloat(Value=1))) {isSetParameters=true}
   /SetVariables(Particles.Color003:LinearColor=&MakeLinearColorFromVectorAndFloat(
     "Vector (RGB)"=&MakeVectorFromLinearColorRGB(Color=<rgba 0.0781 0.4051 1 1>)
     "Float (Alpha)"=&FloatFromCurve(FloatCurve=DataInterface() {propertyValues {Curve {keys [<at 0 0 RCIM_Cubic>,<at 0.4 1 RCIM_Cubic>,<at 1 0 RCIM_Cubic>]}}})
    )) {isSetParameters=true}
   &SolveForcesAndVelocity(Clamp Velocity=false, Limit Acceleration=false)
   &SubUVAnimation(
    End Frame Range Override=3
    "Sprite Renderer "=DataInterface() {propertyValues {SpriteRenderer=&self:Sparkles_0.NiagaraSpriteRendererProperties_3}}
    UseEndFrame=true
    UseStartFrame=true
   )
  ]
  rendererClasses [&NiagaraSpriteRendererProperties]
  renderers [
   {rendererClass=&NiagaraSpriteRendererProperties}
   {rendererClass=&NiagaraSpriteRendererProperties, rendererIndex=1}
  ]
 }
}
summary {
 systemName=NS_Emote_FlossYawn_Frost
 emitters [
  {emitterName=Snow_Trail, rendererClasses [&NiagaraDecalRendererProperties]}
  {emitterName=Secondaries, rendererClasses [&NiagaraSpriteRendererProperties, &NiagaraMeshRendererProperties]}
  {emitterName=Light, rendererClasses [&NiagaraLightRendererProperties]}
  {emitterName=Sparkles, rendererClasses [&NiagaraSpriteRendererProperties]}
 ]
 userVariables [
  User.bIsFrontEnd {type=NiagaraBool}
  User.bIsFrontEndPreview {type=NiagaraBool}
  User.IsFrontEnd {type=NiagaraBool}
  User.LightIntensity {type=NiagaraFloat, defaultValue=0.2, description=""}
  User.Local Alpha Preview {type=NiagaraFloat, defaultValue=0.4, description=""}
  User.Other Alpha Preview {type=NiagaraFloat, defaultValue=0.4, description=""}
  User.OwnerLocalVelocity {type=Vector3f, description="", defaultValue {struct=Vector3f, value=<0 0 0>}}
 ]
}
