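# Niagara system for the FlossYawn emote that draws short-lived, bluish ribbon trails from the player's left and right hand bones.
# A single emitter spawns particles at skeletal mesh bone positions with upward velocity, and a local Leading_Edge_Fix module receives the sampled position.
# Curves fade and taper the trails over particle life.
# Three ribbon renderers share the emitter, alongside separately computed EnergyColor, SnowColor and SnowWidth attributes.
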NiagaraSystem /BRCosmetics/Animation/Game/MainPlayer/Emotes/FlossYawn/FX/FX/NS_Emote_FlossYawn_HandTrails
dependencies {
 usedDataInterfaces [&NiagaraDataInterfaceSkeletalMesh, &NiagaraDataInterfaceCurve, &NiagaraDataInterfaceColorCurve]
 usedDynamicInputs []
 usedModules [&SystemState]
}
emitters {Minimal {
  emitterSpawnScript [
   /SetVariables(Emitter.Noise Panner Speed:Vector4f=<-0.4 0 0 0>) {isSetParameters=true}
   /SetVariables(Emitter.NewSpawnRate=20) {isSetParameters=true}
  ]
  emitterUpdateScript [
   &EmitterState()
   &SpawnRate(SpawnRate=$Emitter.NewSpawnRate, Use Spawn Probability=false)
  ]
  particleSpawnScript [
   &InitializeParticle(
    Color=<rgba 0.8748 1.1847 1.5 1.5>
    Lifetime=0.5
    Ribbon Width=25
    Ribbon Width Mode=::NewEnumerator1
    UsePositionOffset=false
    Write Material Random=true
   )
   &SkeletalMeshLocation(
    Bone Sampling Mode=::NewEnumerator4
    MaskSpawnGroup=false
    Orientation Sampling=::NewEnumerator1
    Position Sampling=::NewEnumerator2
    Skeletal Mesh=DataInterface() {propertyValues {
      DefaultMesh=null
      ExcludeBoneName=root
      OptionalComponentClass=null
      PreviewMesh=null
      SkinningMode=SkinOnTheFly
      SoftSourceActor=null
      SourceMode=Default
      UvSetIndex=0
      WholeMeshLOD=-1
      bExcludeBone=false
      bReadDeformedGeometry=true
      bRequireCurrentFrameData=true
      ComponentTags []
      FilteredBones [Hand_L,Hand_R]
      FilteredSockets []
      MeshUserParameter {parameter {name=None, typeDefHandle {registeredTypeIndex=1}, varData []}}
      SamplingRegions []
     }}
   )
   &AddVelocity(Velocity=<0 0 150>)
   &ApplyInitialForces()
  ]
  particleUpdateScript [
   &ParticleState()
   &FN_Sprite_Attributes(
    Color=&MultiplyLinearColor(
     Linear Color=<rgba 0.3596 0.594 1.2 1>
     Scale Factor=&MakeVector4FromVectorAndFloat(
      "Vector (RGB)"=<1 1 1>
      "Float (Alpha)"=&Multiply_Float(
       A=1
       B=&FloatFromCurve(FloatCurve=DataInterface() {propertyValues {Curve {keys [{arriveTangent=-4.0505, interpMode=RCIM_Cubic, leaveTangent=-4.0505, tangentMode=RCTM_User, time=0.3, value=1}, <at 1 0>]}}})
      )
     )
    )
    Create Color Attribute=true
    Write to Module Namespace=true
   )
   &self:Leading_Edge_Fix(Position:NiagaraPosition=$Particles.SkeletalMeshLocation.SampledPosition) {module=Leading_Edge_Fix}
   &ScaleColor(
    Scale Alpha=&FloatFromCurve(FloatCurve=DataInterface() {propertyValues {Curve {keys [<at 0 0>,<at 0.05 1>,<at 1 0>]}}}, Scale Curve=0.5)
    Scale RGB=&VectorFromFloat(Value=&FloatFromCurve(FloatCurve=DataInterface() {propertyValues {Curve {keys [<at 0 1>,<at 0.3 0.8>]}}}))
    ScaleA=true
    ScaleRGB=true
   )
   &ScaleRibbonWidth(Ribbon Width Scale=&FloatFromCurve(FloatCurve=DataInterface() {propertyValues {Curve {keys [<at 0 1>,<at 0.05 1>,<at 0.8 0>]}}}))
   /SetVariables(Particles.EnergyColor:LinearColor=&MultiplyLinearColor(
     Linear Color=&ColorFromCurve(ColorCurve=DataInterface() {propertyValues {
        AlphaCurve {keys [<at 0 0.8 RCIM_Cubic>,<at 0.6886 0 RCIM_Cubic>]}
        BlueCurve {keys [<at 0 1 RCIM_Cubic>,<at 0.4856 0.875 RCIM_Cubic>]}
        GreenCurve {keys [<at 0 0.9713 RCIM_Cubic>,<at 0.4856 0.5003 RCIM_Cubic>]}
        RedCurve {keys [<at 0 0.7482 RCIM_Cubic>,<at 0.4856 0.4056 RCIM_Cubic>]}
       }})
     Scale Factor=<1 1 1 0.5>
    )) {isSetParameters=true}
   /SetVariables(Particles.SnowWidth:NiagaraFloat=&FloatFromCurve(FloatCurve=DataInterface() {propertyValues {Curve {keys [<at 0 1>,<at 1 0>]}}}, Scale Curve=12)) {isSetParameters=true}
   /SetVariables(Particles.SnowColor:LinearColor=&MultiplyLinearColor(
     Linear Color=&ColorFromCurve(ColorCurve=DataInterface() {propertyValues {
        AlphaCurve {keys [<at 0 0.5 RCIM_Cubic>,<at 0.9052 0 RCIM_Cubic>]}
        BlueCurve {keys [<at 0 10 RCIM_Cubic>,<at 0.3739 1 RCIM_Cubic>]}
        GreenCurve {keys [<at 0 9.0404 RCIM_Cubic>,<at 0.3739 0.8023 RCIM_Cubic>]}
        RedCurve {keys [<at 0 4.1144 RCIM_Cubic>,<at 0.3739 0.5438 RCIM_Cubic>]}
       }})
     Scale Factor=<1 1 1 0.5>
    )) {isSetParameters=true}
  ]
  rendererClasses [&NiagaraRibbonRendererProperties]
  renderers [
   {rendererClass=&NiagaraRibbonRendererProperties}
   {rendererClass=&NiagaraRibbonRendererProperties, rendererIndex=1}
   {rendererClass=&NiagaraRibbonRendererProperties, rendererIndex=2}
  ]
 }}
summary {
 systemName=NS_Emote_FlossYawn_HandTrails
 emitters [{emitterName=Minimal, rendererClasses [&NiagaraRibbonRendererProperties]}]
 userVariables [
  User.Anim_LocalSpaceInheritance {type=NiagaraFloat, defaultValue=0.5, description=""}
  User.EmissionSwitch {type=NiagaraFloat, defaultValue=1, description=""}
  User.OwnerLocalVelocity {type=Vector3f, description="", defaultValue {struct=Vector3f, value=<0 0 0>}}
 ]
}
