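# Niagara intro effect for the FlossYawn player emote: two short ribbon trails wind upward half a turn apart, shedding falling cyan-blue sparkles.
# Emitter age drives each ribbon's spawn angle, height, shrinking radius and growing width over one brief burst, tracing converging helices without velocity forces.
# Two sprite emitters spawn from the ribbon particles via attribute readers, each with three renderers plus animated line-scale and glow variables.
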NiagaraSystem /BRCosmetics/Animation/Game/MainPlayer/Emotes/FlossYawn/FX/FX/NS_Emote_FlossYawn_Intro
dependencies {
 usedDataInterfaces [
  &NiagaraDataInterfaceParticleRead
  &NiagaraDataInterfaceCurve
  &NiagaraDataInterfaceVectorCurve
  &NiagaraDataInterfaceColorCurve
 ]
 usedDynamicInputs []
 usedModules [&SystemState]
}
emitters {
 Minimal {
  emitterUpdateScript [
   &EmitterState(Life Cycle Mode=::NewEnumerator1, Loop Behavior=::NewEnumerator1, Loop Duration=0.4, UseLoopDelay=false)
   &SpawnRate(
    SpawnRate=&FloatFromCurve(
     FloatCurve=DataInterface() {propertyValues {Curve {
        defaultValue=1
        keys [
         {tangentMode=RCTM_User, time=0, value=1}
         {arriveTangent=-0.0405, interpMode=RCIM_Cubic, leaveTangent=-0.0405, tangentMode=RCTM_User, time=0.98, value=1}
         {arriveTangent=-10.8347, interpMode=RCIM_Cubic, leaveTangent=-10.8347, tangentMode=RCTM_User, time=1, value=0}
        ]
       }}}
     CurveIndex=$Emitter.NormalizedLoopAge
     Scale Curve=150
    )
    Use Spawn Probability=false
   )
  ]
  particleSpawnScript [
   &InitializeParticle(
    Color Mode=::NewEnumerator0
    Lifetime=0.5
    Ribbon Width=&FloatFromCurve(
     FloatCurve=DataInterface() {propertyValues {Curve {defaultValue=0, keys [{tangentMode=RCTM_User, time=0, value=0}, {tangentMode=RCTM_User, time=1, value=1}]}}}
     CurveIndex=&Divide_Float(A=$Emitter.Age, B=0.4)
     Scale Curve=15
    )
    Ribbon Width Mode=::NewEnumerator1
    UsePositionOffset=false
    Write Material Random=true
   )
   &ShapeLocation(
    MaskSpawnGroup=false
    Offset=/Niagara/DynamicInputs/Math/Clamp_Vector(
     Vector:Vector3f=&Lerp_Vector(A=<0 0 0>, B=<0 0 220>, Alpha=&Divide_Float(A=$Emitter.Age, B=0.4))
     Min:Vector3f=<0 0 0>
     Max:Vector3f=<0 0 220>
    )
    Offset Mode=::NewEnumerator0
    Override Seed=false
    Radius Position:NiagaraFloat=&FloatFromCurve(
     FloatCurve=DataInterface() {propertyValues {Curve {keys [<at 0 1>,<at 1 0>]}}}
     CurveIndex=&Divide_Float(A=$Emitter.Age, B=0.4)
    )
    Ring / Disc Distribution Mode:ENiagaraSphereDistributionMode=::NewEnumerator1
    Shape Primitive=::NewEnumerator4
    U Position=&Divide_Float(A=$Emitter.Age, B=0.4)
    UseOwnerScale=true
   )
  ]
  particleUpdateScript [
   &ParticleState()
   &ScaleColor(
    Scale Alpha=&FloatFromCurve(FloatCurve=DataInterface() {propertyValues {Curve {keys [<at 0 0>,<at 0.0982 1>,<at 1 0>]}}})
    ScaleA=true
    ScaleRGB=true
   )
   &ScaleRibbonWidth(Ribbon Width Scale=&FloatFromCurve(FloatCurve=DataInterface() {propertyValues {Curve {keys [<at 0 2>,<at 0.2663 0.9>,<at 1 0>]}}}))
  ]
  rendererClasses [&NiagaraRibbonRendererProperties]
  renderers [{rendererClass=&NiagaraRibbonRendererProperties}]
 }
 Minimal003 {
  emitterUpdateScript [
   &EmitterState()
   &SpawnParticlesFromOtherEmitter(
    Attribute Reader=DataInterface() {propertyValues {EmitterBinding {bindingMode=Other, emitterName=Minimal004}}}
    Maximum Age Spawn Threshold=0.5
    Spawn Rate=&FloatFromCurve(
     FloatCurve=DataInterface() {propertyValues {Curve {keys [<at 0 1>,<at 0.4898 2.998>,<at 1 2.5>]}}}
     CurveIndex=&Divide_Float(A=$Emitter.Age, B=1)
     Scale Curve=7
    )
    Spawn Rate Per Particle Cap=10
    Use Maximum Age=true
    Use Minimum Age=false
   )
  ]
  particleSpawnScript [
   &InitializeParticle(
    Color=<rgba 0.6971 1.6118 2 1>
    Sprite Size Mode=::NewEnumerator3
    Uniform Sprite Size=14
    UsePositionOffset=false
    Write Material Random=true
   )
   &SampleParticlesFromOtherEmitter()
  ]
  particleUpdateScript [
   &ParticleState()
   &ScaleSpriteSize(Uniform Curve Sprite Scale=DataInterface() {propertyValues {Curve {defaultValue=0, keys [{tangentMode=RCTM_User, time=0, value=1}, {tangentMode=RCTM_User, time=1, value=0}]}}})
   &GravityForce(Gravity=<0 0 -50>)
   &SolveForcesAndVelocity(Clamp Velocity=false, Limit Acceleration=false)
   &ScaleColor(
    Scale Alpha=&FloatFromCurve(FloatCurve=DataInterface() {propertyValues {Curve {
        defaultValue=1
        keys [
         {arriveTangent=9.0725, interpMode=RCIM_Cubic, leaveTangent=9.0725, tangentMode=RCTM_User, time=0, value=0}
         {arriveTangent=-2.5249, interpMode=RCIM_Cubic, leaveTangent=-2.5249, tangentMode=RCTM_User, time=0.1408, value=0.9847}
         {arriveTangent=-0.0405, interpMode=RCIM_Cubic, leaveTangent=-0.0405, tangentMode=RCTM_User, time=1, value=0}
        ]
       }}})
    Scale RGB=&VectorFromCurve(VectorCurve=DataInterface() {propertyValues {XCurve {keys [<at 0 1>,<at 1 0>]}, YCurve {keys [<at 0 1>,<at 1 0>]}, ZCurve {keys [<at 0 1>,<at 1 0.5>]}}})
    ScaleA=true
    ScaleRGB=true
   )
   /SetVariables(Particles.LineScale:Vector2f=&Lerp_Vector2(
     A=<5 80>
     B=<2 40>
     Alpha=&FloatFromCurve(
      FloatCurve=DataInterface() {propertyValues {Curve {defaultValue=1, keys [{tangentMode=RCTM_User, time=0, value=0}, <at 0.25 0>, {tangentMode=RCTM_User, time=0.8, value=1}]}}}
      CurveIndex=$Emitter.Age
     )
    )) {isSetParameters=true, module=SetVariables_#1}
   /SetVariables(Particles.GlowColor=&ColorFromCurve(ColorCurve=DataInterface() {propertyValues {
       AlphaCurve {keys [<at 0 0 RCIM_Cubic>,<at 0.0401 0.02>,<at 1 0 RCIM_Cubic>]}
       BlueCurve {keys [<at 0 1 RCIM_Cubic>,<at 0.7946 0.7865 RCIM_Cubic>]}
       GreenCurve {keys [<at 0 0.612 RCIM_Cubic>,<at 0.7946 0.1117 RCIM_Cubic>]}
       RedCurve {keys [<at 0 0.0513 RCIM_Cubic>,<at 0.7946 0.0232 RCIM_Cubic>]}
      }})) {isSetParameters=true, module=SetVariables_#2}
   /SetVariables(Particles.GlowScale:Vector2f=<25 25>) {isSetParameters=true, module=SetVariables_#3}
   /SetVariables() {isSetParameters=true}
  ]
  rendererClasses [&NiagaraSpriteRendererProperties]
  renderers [
   {rendererClass=&NiagaraSpriteRendererProperties}
   {rendererClass=&NiagaraSpriteRendererProperties, rendererIndex=1}
   {rendererClass=&NiagaraSpriteRendererProperties, rendererIndex=2}
  ]
 }
 Minimal004 {
  emitterUpdateScript [
   &EmitterState(Life Cycle Mode=::NewEnumerator1, Loop Behavior=::NewEnumerator1, Loop Duration=0.4, UseLoopDelay=false)
   &SpawnRate(
    SpawnRate=&FloatFromCurve(
     FloatCurve=DataInterface() {propertyValues {Curve {
        defaultValue=1
        keys [
         {tangentMode=RCTM_User, time=0, value=1}
         {arriveTangent=-0.0405, interpMode=RCIM_Cubic, leaveTangent=-0.0405, tangentMode=RCTM_User, time=0.98, value=1}
         {arriveTangent=-10.8347, interpMode=RCIM_Cubic, leaveTangent=-10.8347, tangentMode=RCTM_User, time=1, value=0}
        ]
       }}}
     CurveIndex=$Emitter.NormalizedLoopAge
     Scale Curve=150
    )
    Use Spawn Probability=false
   )
  ]
  particleSpawnScript [
   &InitializeParticle(
    Color Mode=::NewEnumerator0
    Lifetime=0.5
    Ribbon Width=&FloatFromCurve(
     FloatCurve=DataInterface() {propertyValues {Curve {defaultValue=0, keys [{tangentMode=RCTM_User, time=0, value=0}, {tangentMode=RCTM_User, time=1, value=1}]}}}
     CurveIndex=&Divide_Float(A=$Emitter.Age, B=0.4)
     Scale Curve=15
    )
    Ribbon Width Mode=::NewEnumerator1
    UsePositionOffset=false
    Write Material Random=true
   )
   &ShapeLocation(
    MaskSpawnGroup=false
    Offset=&Clamp_Vector(
     Vector:Vector3f=&Lerp_Vector(A=<0 0 0>, B=<0 0 220>, Alpha=&Divide_Float(A=$Emitter.Age, B=0.4))
     Min:Vector3f=<0 0 0>
     Max:Vector3f=<0 0 220>
    )
    Offset Mode=::NewEnumerator0
    Override Seed=false
    Radius Position:NiagaraFloat=&FloatFromCurve(
     FloatCurve=DataInterface() {propertyValues {Curve {keys [<at 0 1>,<at 1 0>]}}}
     CurveIndex=&Divide_Float(A=$Emitter.Age, B=0.4)
    )
    Ring / Disc Distribution Mode:ENiagaraSphereDistributionMode=::NewEnumerator1
    Shape Primitive=::NewEnumerator4
    U Position=&Add_Float(A=&Divide_Float(A=$Emitter.Age, B=0.4), B=0.5)
    UseOwnerScale=true
   )
  ]
  particleUpdateScript [
   &ParticleState()
   &ScaleColor(
    Scale Alpha=&FloatFromCurve(FloatCurve=DataInterface() {propertyValues {Curve {keys [<at 0 0>,<at 0.0982 1>,<at 1 0>]}}})
    ScaleA=true
    ScaleRGB=true
   )
   &ScaleRibbonWidth(Ribbon Width Scale=&FloatFromCurve(FloatCurve=DataInterface() {propertyValues {Curve {keys [<at 0 2>,<at 0.2747 0.8059>,<at 1 0>]}}}))
  ]
  rendererClasses [&NiagaraRibbonRendererProperties]
  renderers [{rendererClass=&NiagaraRibbonRendererProperties}]
 }
 Minimal005 {
  emitterUpdateScript [
   &EmitterState()
   &SpawnParticlesFromOtherEmitter(
    Attribute Reader=DataInterface() {propertyValues {EmitterBinding {bindingMode=Other, emitterName=Minimal}}}
    Maximum Age Spawn Threshold=0.5
    Spawn Rate=&FloatFromCurve(
     FloatCurve=DataInterface() {propertyValues {Curve {keys [<at 0 1>,<at 0.4898 2.998>,<at 1 2.5>]}}}
     CurveIndex=&Divide_Float(A=$Emitter.Age, B=1)
     Scale Curve=7
    )
    Spawn Rate Per Particle Cap=10
    Use Maximum Age=true
    Use Minimum Age=false
   )
  ]
  particleSpawnScript [
   &InitializeParticle(
    Color=<rgba 0.6971 1.6118 2 1>
    Sprite Size Mode=::NewEnumerator3
    Uniform Sprite Size=14
    UsePositionOffset=false
    Write Material Random=true
   )
   &SampleParticlesFromOtherEmitter()
  ]
  particleUpdateScript [
   &ParticleState()
   &ScaleSpriteSize(Uniform Curve Sprite Scale=DataInterface() {propertyValues {Curve {defaultValue=0, keys [{tangentMode=RCTM_User, time=0, value=1}, {tangentMode=RCTM_User, time=1, value=0}]}}})
   &GravityForce(Gravity=<0 0 -50>)
   &SolveForcesAndVelocity(Clamp Velocity=false, Limit Acceleration=false)
   &ScaleColor(
    Scale Alpha=&FloatFromCurve(FloatCurve=DataInterface() {propertyValues {Curve {
        defaultValue=1
        keys [
         {arriveTangent=9.0725, interpMode=RCIM_Cubic, leaveTangent=9.0725, tangentMode=RCTM_User, time=0, value=0}
         {arriveTangent=-2.5249, interpMode=RCIM_Cubic, leaveTangent=-2.5249, tangentMode=RCTM_User, time=0.1408, value=0.9847}
         {arriveTangent=-0.0405, interpMode=RCIM_Cubic, leaveTangent=-0.0405, tangentMode=RCTM_User, time=1, value=0}
        ]
       }}})
    Scale RGB=&VectorFromCurve(VectorCurve=DataInterface() {propertyValues {XCurve {keys [<at 0 1>,<at 1 0>]}, YCurve {keys [<at 0 1>,<at 1 0>]}, ZCurve {keys [<at 0 1>,<at 1 0.5>]}}})
    ScaleA=true
    ScaleRGB=true
   )
   /SetVariables(Particles.LineScale:Vector2f=&Lerp_Vector2(
     A=<5 80>
     B=<2 40>
     Alpha=&FloatFromCurve(
      FloatCurve=DataInterface() {propertyValues {Curve {defaultValue=1, keys [{tangentMode=RCTM_User, time=0, value=0}, <at 0.25 0>, {tangentMode=RCTM_User, time=0.8, value=1}]}}}
      CurveIndex=$Emitter.Age
     )
    )) {isSetParameters=true, module=SetVariables_#1}
   /SetVariables(Particles.GlowColor=&ColorFromCurve(ColorCurve=DataInterface() {propertyValues {
       AlphaCurve {keys [<at 0 0 RCIM_Cubic>,<at 0.0401 0.02>,<at 1 0 RCIM_Cubic>]}
       BlueCurve {keys [<at 0 1 RCIM_Cubic>,<at 0.7946 0.7865 RCIM_Cubic>]}
       GreenCurve {keys [<at 0 0.612 RCIM_Cubic>,<at 0.7946 0.1117 RCIM_Cubic>]}
       RedCurve {keys [<at 0 0.0513 RCIM_Cubic>,<at 0.7946 0.0232 RCIM_Cubic>]}
      }})) {isSetParameters=true, module=SetVariables_#2}
   /SetVariables(Particles.GlowScale:Vector2f=<25 25>) {isSetParameters=true, module=SetVariables_#3}
   /SetVariables() {isSetParameters=true}
  ]
  rendererClasses [&NiagaraSpriteRendererProperties]
  renderers [
   {rendererClass=&NiagaraSpriteRendererProperties}
   {rendererClass=&NiagaraSpriteRendererProperties, rendererIndex=1}
   {rendererClass=&NiagaraSpriteRendererProperties, rendererIndex=2}
  ]
 }
}
summary {
 systemName=NS_Emote_FlossYawn_Intro
 emitters [
  {emitterName=Minimal003, rendererClasses [&NiagaraSpriteRendererProperties]}
  {emitterName=Minimal, rendererClasses [&NiagaraRibbonRendererProperties]}
  {emitterName=Minimal004, rendererClasses [&NiagaraRibbonRendererProperties]}
  {emitterName=Minimal005, rendererClasses [&NiagaraSpriteRendererProperties]}
 ]
 userVariables [
  User.bIsFrontEnd {type=NiagaraBool}
  User.bIsFrontEndPreview {type=NiagaraBool}
  User.bisHeadHidden {type=NiagaraBool}
  User.MusicReactive {type=NiagaraBool, defaultValue=true}
 ]
}
