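# Niagara system for a snowflake burst in the FlossYawn emote, though all four real emitters are disabled and only five empty placeholders are enabled.
# Minimal bursts an elevated glow sprite, which DirectionalBurst samples through a particle attribute reader to spawn small spinning SubUV sprites.
# NE_First_Flare adds an intro-delayed flare tinted by time of day.
# HeartsSparkles001 bursts stretched sparkles outward with drag and flickering bokeh parameters.
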NiagaraSystem /BRCosmetics/Animation/Game/MainPlayer/Emotes/FlossYawn/FX/FX/NS_Emote_FlossYawn_SnowflakeBurst
dependencies {
 usedDataInterfaces [
  &NiagaraDataInterfaceCurve
  &NiagaraDataInterfaceColorCurve
  &NiagaraDataInterfaceArrayFloat3
  &NiagaraDataInterfaceVector2DCurve
  &NiagaraDataInterfaceParticleRead
  &NiagaraDataInterfaceSpriteRendererInfo
 ]
 usedDynamicInputs [&MakeCustomLinearColorFromBool]
 usedModules [
  &self:SystemSpawnScript.SystemScriptSource.SystemScriptGraph.NiagaraNodeAssignment_0.SetVariables_#1
  &SystemState
 ]
}
emitters {
 DirectionalBurst {
  bEnabled=false
  emitterUpdateScript [
   &EmitterState()
   &SpawnParticlesFromOtherEmitter(
    Attribute Reader=DataInterface() {propertyValues {EmitterBinding {bindingMode=Other, emitterName=Minimal}}}
    Spawn Rate=15
    Use Maximum Age=false
    Use Minimum Age=false
   )
  ]
  particleSpawnScript [
   &InitializeParticle(
    Color=<rgba 0.1234 0.6974 1 1>
    Lifetime=0.2
    Position Mode=::NewEnumerator1
    Sprite Rotation Angle=&RandomRangeFloat(Minimum=0, Maximum=0)
    Sprite Rotation Mode=::NewEnumerator1
    Sprite Size Mode=::NewEnumerator3
    Uniform Sprite Size=20
    UsePositionOffset=false
    Write Material Random=true
   )
   &CameraOffset()
   &SampleParticlesFromOtherEmitter()
   &ShapeLocation(MaskSpawnGroup=false, Override Seed=false, Sphere Radius=20, UseOwnerScale=true)
  ]
  particleUpdateScript [
   &ParticleState()
   &SpriteRotationRate(Rotation Rate=&Multiply_Float(
     A=&RandomRangeFloat(Minimum=-10, Maximum=10)
     B=&FloatFromCurve(FloatCurve=DataInterface() {propertyValues {Curve {keys [<at 0 1>,<at 1 0>]}}}, Scale Curve=10)
    ))
   &ScaleSpriteSize(
    Uniform Curve Scale=&RandomRangeFloat(Minimum=0.6, Maximum=1)
    Uniform Curve Sprite Scale=DataInterface() {propertyValues {Curve {keys [{tangentMode=RCTM_User, time=0, value=0.4}, <at 0.1 1.2>, <at 0.2 1>, {tangentMode=RCTM_User, time=1, value=1}]}}}
   )
   /SetVariables(Particles.Size002:Vector2f=&Multiply_Vector2D(
     A=$Particles.SpriteSize
     B=&Vector2DFromFloat(Value=&RandomRangeFloat(Minimum=0.5, Maximum=0.8))
    )) {isSetParameters=true}
   &ScaleColor(
    Scale Alpha=&FloatFromCurve(FloatCurve=DataInterface() {propertyValues {Curve {keys [<at -0.0046 1.0113 RCIM_Cubic>,<at 1.0011 0.002 RCIM_Cubic>]}}})
    ScaleA=true
    ScaleRGB=true
   )
   /SetVariables(Particles.Size003:Vector2f=&Multiply_Vector2D(A=<10 10>, B=&Vector2DFromFloat(Value=1))) {isSetParameters=true}
   /SetVariables(Particles.Color003:LinearColor=&MakeLinearColorFromVectorAndFloat(
     "Vector (RGB)"=&MakeVectorFromLinearColorRGB(Color=<rgba 0.0781 0.4051 1 1>)
     "Float (Alpha)"=&FloatFromCurve(FloatCurve=DataInterface() {propertyValues {Curve {keys [<at 0 0 RCIM_Cubic>,<at 0.4 1 RCIM_Cubic>,<at 1 0 RCIM_Cubic>]}}})
    )) {isSetParameters=true}
   &SolveForcesAndVelocity(Clamp Velocity=false, Limit Acceleration=false)
   &SubUVAnimation(
    End Frame Range Override=3
    "Sprite Renderer "=DataInterface() {propertyValues {SpriteRenderer=&self:DirectionalBurst_0.NiagaraSpriteRendererProperties_3}}
    UseEndFrame=true
    UseStartFrame=true
   )
   /SetVariables(Particles.Color_glow:LinearColor=&MakeLinearColorFromVectorAndFloat(
     "Vector (RGB)"=&MakeVectorFromLinearColorRGB(Color=<rgba 1 1 1 1>)
     "Float (Alpha)"=&FloatFromCurve(
      FloatCurve=DataInterface() {propertyValues {Curve {keys [<at 0.5 1>,<at 1 0>]}}}
      Scale Curve=&MakeCustomFloatFromBool(
       Bool=true
       True Float=2
       False Float=&MakeCustomFloatFromBool(
        Bool=&SetBoolByFloatComparison(A=$System.Music_Reactive_FX_Preview, B=0)
        True Float=&Multiply_Float(
         A=&FN_FrontEndFloatModifier(Gameplay=2, Front End=0.3)
         B=&FNTimeOfDayScalar(Night Scale=0.1, Morning Scale=1, Evening Scale=1)
        )
        False Float=0
       )
      )
     )
    )) {enabled=false, isSetParameters=true}
  ]
  rendererClasses [&NiagaraSpriteRendererProperties]
  renderers [{rendererClass=&NiagaraSpriteRendererProperties}]
 }
 Empty {rendererClasses [], renderers []}
 Empty001 {rendererClasses [], renderers []}
 Empty002 {rendererClasses [], renderers []}
 Empty003 {rendererClasses [], renderers []}
 Empty004 {rendererClasses [], renderers []}
 HeartsSparkles001 {
  bEnabled=false
  emitterUpdateScript [
   &EmitterState()
   &SpawnBurst_Instantaneous(Spawn Count=30, Spawn Time=0.58, Use Spawn Probability=false, UseLoopCountLimit=false)
   /SetVariables(Emitter.Heart Color:LinearColor=<rgba 2 0.3055 0.6324 1>) {isSetParameters=true}
  ]
  particleSpawnScript [
   &InitializeParticle(
    Color=&MergeHelper_ColorSwitch(
     0=&RandomRangeLinearColor(Minimum=<rgba 2 0.6743 0.9242 1>, Maximum=<rgba 2 1.8975 1.8783 1>)
     1=&RandomRangeLinearColor(Minimum=<rgba 3 1.681 2.0591 1>, Maximum=<rgba 5 4.5283 4.7799 1>)
     2=?
     3=?
     Index=&SpawnGroup()
     Num=::NewEnumerator0
    )
    Lifetime=&RandomRangeFloat(Minimum=0.2, Maximum=0.4)
    Position Offset=<0 0 220>
    Sprite Rotation Mode=::NewEnumerator2
    Sprite Size=<4 80>
    Sprite Size Mode=::NewEnumerator4
    UsePositionOffset=true
    Write Material Random=true
   )
   &ShapeLocation(
    MaskSpawnGroup=false
    Non Uniform Scale=<1.5 1 1>
    Offset Mode=::NewEnumerator0
    Override Seed=false
    Sphere Radius=1
    Sphere Surface Distribution=1
    UseOwnerScale=true
   )
   &AddVelocity(Velocity=&RandomRangeVector(Minimum=<-30 -30 100>, Maximum=<30 30 300>)) {enabled=false, module=AddVelocity001}
   &AddVelocity(Module.Use Falloff Distance=false, Origin Offset=<0 0 220>, Override Seed=false, Velocity Mode=::NewEnumerator1, Velocity Speed=175)
  ]
  particleUpdateScript [
   &ParticleState()
   &ScaleSpriteSize(
    Non-Uniform Curve Sprite Scale=DataInterface() {propertyValues {
      XCurve {keys [{tangentMode=RCTM_User, time=0, value=1}, {tangentMode=RCTM_User, time=1, value=1}]}
      YCurve {keys [{tangentMode=RCTM_User, time=0, value=1}, {tangentMode=RCTM_User, time=0.45, value=0.3}]}
     }}
    Scale Sprite Size Mode=::NewEnumerator3
   )
   /SetVariables(
    Particles.Bokeh Flicker:NiagaraFloat=&Waveform(
     "Waveform [1]"=::NewEnumerator0
     "[1] Frequency"=&RandomRangeFloat(Minimum=1, Maximum=5)
     "[1] Amplitude Min/Max"=<0.1 1>
    )
    Particles.Flicker:NiagaraFloat=&Waveform(
     "Waveform [1]"=::NewEnumerator0
     "[1] Frequency"=&RandomRangeFloat(Minimum=3, Maximum=5)
     "[1] Amplitude Min/Max"=<0 1>
    )
   ) {isSetParameters=true}
   &Drag(
    Drag=&Multiply_Float(
     A=&RandomRangeFloat(Minimum=5, Maximum=10)
     B=&FloatFromCurve(
      FloatCurve=DataInterface() {propertyValues {Curve {defaultValue=1, keys [{tangentMode=RCTM_User, time=0.2, value=0}, {tangentMode=RCTM_User, time=0.65, value=3}]}}}
      Scale Curve=2
     )
    )
    Use Linear Drag=true
    Use Rotational Drag=true
   )
   &CurlNoiseForce(
    Bias Noise=false
    Noise Frequency=200
    Noise Strength=&RandomRangeFloat(Minimum=50, Maximum=100)
    Pan Noise=false
   ) {enabled=false}
   &SolveForcesAndVelocity(Clamp Velocity=false, Limit Acceleration=false)
   &ScaleColor(
    Scale Alpha=&FloatFromCurve(FloatCurve=DataInterface() {propertyValues {Curve {
        defaultValue=1
        keys [
         {interpMode=RCIM_Cubic, tangentMode=RCTM_User, time=0, value=0}
         {interpMode=RCIM_Cubic, tangentMode=RCTM_User, time=0.5, value=1}
         {tangentMode=RCTM_User, time=1, value=0}
        ]
       }}})
    Scale RGB=&MergeHelper_VectorSwitch(
     0=<1 1 1>
     1=&MakeVectorFromLinearColorRGB(Color=&ColorFromCurve(ColorCurve=DataInterface() {propertyValues {
         AlphaCurve {keys [<at 1 0>]}
         BlueCurve {keys [<at 0.0637 3>,<at 0.1009 1>,<at 0.3329 1>,<at 0.9944 0.1323>]}
         GreenCurve {keys [<at 0.0637 3>,<at 0.1009 1>,<at 0.3329 1>,<at 0.9944 0>]}
         RedCurve {keys [<at 0.0637 3>,<at 0.1009 1>,<at 0.3329 1>,<at 0.9944 0.3646>]}
        }}))
     2=?
     3=?
     Index=&SpawnGroup()
     Num=::NewEnumerator0
    )
    ScaleA=true
    ScaleRGB=true
   )
   /SetVariables(
    Particles.Bokeh Color:LinearColor=&MakeLinearColorFromVectorAndFloat(
     "Vector (RGB)"=&MakeVectorFromLinearColorRGB(Color=<rgba 0.9909 1.6413 2 1>)
     "Float (Alpha)"=&Multiply_Float(
      A=&FloatFromCurve(
       FloatCurve=DataInterface() {propertyValues {Curve {
          defaultValue=1
          keys [
           {arriveTangent=-1.2032, interpMode=RCIM_Cubic, leaveTangent=-1.2032, tangentMode=RCTM_User, time=0, value=1.2}
           {arriveTangent=-1.2, interpMode=RCIM_Cubic, leaveTangent=-1.2, time=0.3864, value=0.1909}
           {interpMode=RCIM_Cubic, tangentMode=RCTM_User, time=1, value=0}
          ]
         }}}
       Scale Curve=0.3
      )
      B=$Particles.Bokeh Flicker
     )
    )
    Particles.Bokeh Size:Vector2f=&Multiply_Vector2D_ByFloat(
     Float=&FloatFromCurve(FloatCurve=DataInterface() {propertyValues {Curve {
         defaultValue=1
         keys [
          {arriveTangent=40.8712, interpMode=RCIM_Cubic, leaveTangent=40.8712, tangentMode=RCTM_User, time=0, value=0}
          {arriveTangent=0.7672, interpMode=RCIM_Cubic, leaveTangent=0.7672, tangentMode=RCTM_User, time=0.0663, value=1.0039}
          {arriveTangent=-1.712, interpMode=RCIM_Cubic, leaveTangent=-1.712, tangentMode=RCTM_User, time=0.3106, value=0.656}
          {arriveTangent=-0.0405, interpMode=RCIM_Cubic, leaveTangent=-0.0405, tangentMode=RCTM_User, time=1, value=0.25}
         ]
        }}})
     Vector2D=&Vector2DFromFloat(Value=&RandomRangeFloat(Minimum=5, Maximum=25))
    )
   ) {isSetParameters=true}
   /SetVariables(Particles.Vector 4:Vector4f=&MakeVector4(
     X=0
     Y=0
     Z=&Multiply_Float(
      A=1
      B=&FloatFromCurve(FloatCurve=DataInterface() {propertyValues {Curve {
          defaultValue=1
          keys [
           {arriveTangent=24.9349, interpMode=RCIM_Cubic, leaveTangent=24.9349, tangentMode=RCTM_User, time=0, value=0}
           {arriveTangent=0.7276, interpMode=RCIM_Cubic, leaveTangent=0.7276, tangentMode=RCTM_User, time=0.0783, value=0.9655}
           {arriveTangent=-1.8102, interpMode=RCIM_Cubic, leaveTangent=-1.8102, tangentMode=RCTM_User, time=0.6189, value=0.8021}
           {arriveTangent=-2.1697, interpMode=RCIM_Cubic, leaveTangent=-2.1697, tangentMode=RCTM_User, time=0.8033, value=0.3088}
           {arriveTangent=-0.0405, interpMode=RCIM_Cubic, leaveTangent=-0.0405, tangentMode=RCTM_User, time=1, value=0}
          ]
         }}})
     )
     W=0
    )) {enabled=false, isSetParameters=true}
  ]
  rendererClasses [&NiagaraSpriteRendererProperties]
  renderers [{rendererClass=&NiagaraSpriteRendererProperties}]
 }
 Minimal {
  bEnabled=false
  emitterSpawnScript [/SetVariables(Emitter.EmissiveAdjustment=1) {isSetParameters=true}]
  emitterUpdateScript [
   &EmitterState()
   &SpawnBurst_Instantaneous(Use Spawn Probability=false, UseLoopCountLimit=false)
  ]
  particleSpawnScript [&InitializeParticle(
    Color=<rgba 0.451 0.9408 1.3 1>
    Lifetime=0.6
    Position Offset=<0 0 220>
    Sprite Size Mode=::NewEnumerator3
    Uniform Sprite Size=25
    UsePositionOffset=true
    Write Material Random=true
   )]
  particleUpdateScript [
   &ParticleState()
   &ScaleSpriteSize(Uniform Curve Sprite Scale=DataInterface() {propertyValues {Curve {keys [{tangentMode=RCTM_User, time=0, value=0.5}, {tangentMode=RCTM_User, time=0.3, value=1.2}, <at 1.0044 1.5>]}}})
   &SolveForcesAndVelocity(Clamp Velocity=false, Limit Acceleration=false)
   /SetVariables(Particles.GlowColor=&MultiplyLinearColor(
     Linear Color=&ColorFromCurve(ColorCurve=DataInterface() {propertyValues {
        AlphaCurve {keys [<at 0 0 RCIM_Cubic>,<at 0.0356 1>,<at 1 0 RCIM_Cubic>]}
        BlueCurve {keys [<at 0 1 RCIM_Cubic>,<at 1 1 RCIM_Cubic>]}
        GreenCurve {keys [<at 0 0.4764 RCIM_Cubic>,<at 1 0.1845 RCIM_Cubic>]}
        RedCurve {keys [<at 0 0.0859 RCIM_Cubic>,<at 1 0 RCIM_Cubic>]}
       }})
     Scale Factor=<1 1 1 0.5>
    )) {isSetParameters=true}
   /SetVariables(Particles.GlowScale:Vector2f=<50 50>) {isSetParameters=true}
   /SetVariables(Particles.LineScale:Vector2f=<360 5>) {isSetParameters=true}
   /SetVariables(Particles.Linear Color=<rgba 1 1 1 0.05>) {isSetParameters=true}
  ]
  rendererClasses [&NiagaraSpriteRendererProperties]
  renderers [
   {rendererClass=&NiagaraSpriteRendererProperties}
   {rendererClass=&NiagaraSpriteRendererProperties, rendererIndex=1}
  ]
 }
 NE_First_Flare {
  bEnabled=false
  emitterUpdateScript [
   &EmitterState(Life Cycle Mode=::NewEnumerator1, Loop Behavior=::NewEnumerator1, Loop Delay=$System.IntroLength, Loop Duration=2, UseLoopDelay=true)
   &SpawnBurst_Instantaneous(Spawn Time=0.6, Use Spawn Probability=false, UseLoopCountLimit=false)
  ]
  particleSpawnScript [
   &InitializeParticle(
    Color=<rgba 0.0509 0.5044 0.6979 1>
    Lifetime=0.5
    Mass Mode=::NewEnumerator1
    Mesh Scale Mode=::NewEnumerator3
    Mesh Uniform Scale=&FloatFromCurve(
     FloatCurve=DataInterface() {propertyValues {Curve {keys [{tangentMode=RCTM_User, time=0, value=1}, {tangentMode=RCTM_User, time=1, value=2}]}}}
     CurveIndex=&ReturnNormalizedExecIndex()
     Scale Curve=1.5
    )
    Position Offset=<0 0 220>
    Sprite Size=<75 75>
    Sprite Size Mode=::NewEnumerator4
    UsePositionOffset=true
    Write Material Random=true
   ) {module=InitializeParticle001}
   &CameraOffset(Camera Offset Amount=5)
   &InitialMeshOrientation(
    Mesh Orientation Mode=::NewEnumerator2
    Rotation=&SelectVectorFromArray(
     Direct Array Index=&SpawnGroup()
     Vector Selection Array=DataInterface() {propertyValues {FloatData [<-0.25 0 0>, <0.5 0 0>]}}
    )
    Rotation Coordinate Space=::NewEnumerator2
    Use Rotation Vector=true
   )
  ]
  particleUpdateScript [
   &ParticleState()
   &SolveForcesAndVelocity(Clamp Velocity=false, Limit Acceleration=false)
   &ScaleColor(
    Scale Alpha=&FloatFromCurve(
     FloatCurve=DataInterface() {propertyValues {Curve {keys [
         {arriveTangent=6.4019, interpMode=RCIM_Cubic, leaveTangent=6.4019, tangentMode=RCTM_User, time=0, value=0}
         {arriveTangent=0.002, interpMode=RCIM_Cubic, leaveTangent=0.002, time=0.2, value=0.7}
         {arriveTangent=-0.9006, interpMode=RCIM_Cubic, leaveTangent=-0.9006, tangentMode=RCTM_User, time=1.0011, value=0.002}
        ]}}}
     Scale Curve=&FNTimeOfDayScalar(Night Scale=0.5, Morning Scale=0.7, Evening Scale=0.7)
    )
    Scale RGB=&MakeVectorFromLinearColorRGB(Color=&FNTimeOfDayColor(Night Color=<rgba 0.5006 0.8335 1 1>, Morning Color=<rgba 1 1 1 1>, Day Color=<rgba 1 1 1 1>, Evening Color=<rgba 1 1 1 1>))
    ScaleA=true
    ScaleRGB=true
   )
   &ScaleSpriteSize(Scale Factor=&Multiply_Vector2D_ByFloat(
     Float=&FloatFromCurve(
      FloatCurve=DataInterface() {propertyValues {Curve {keys [
          {arriveTangent=0.7585, interpMode=RCIM_Cubic, leaveTangent=0.7585, tangentMode=RCTM_User, time=0, value=0.4}
          {interpMode=RCIM_Cubic, tangentMode=RCTM_User, time=1, value=1}
         ]}}}
      Scale Curve=2
     )
     Vector2D=<1 1>
    ))
   &ScaleMeshSize(Scale Factor=&VectorFromFloat(Value=&FloatFromCurve(
      FloatCurve=DataInterface() {propertyValues {Curve {keys [
          {arriveTangent=4.1785, interpMode=RCIM_Cubic, leaveTangent=4.1785, tangentMode=RCTM_User, time=0, value=0}
          {arriveTangent=1.1179, interpMode=RCIM_Cubic, leaveTangent=1.1179, tangentMode=RCTM_User, time=0.2042, value=0.6228}
          {arriveTangent=0.0224, interpMode=RCIM_Cubic, leaveTangent=0.0224, tangentMode=RCTM_User, time=1, value=1}
         ]}}}
      Scale Curve=0.35
     )))
   /SetVariables(Particles.CentralFlareSize:Vector2f=&Multiply_Vector2D_ByFloat(Float=1, Vector2D=$Particles.SpriteSize)) {isSetParameters=true}
   /SetVariables(Particles.CentralFlareColor:LinearColor=&MultiplyLinearColor(
     Linear Color=&ColorFromCurve(
      ColorCurve=DataInterface() {propertyValues {
        AlphaCurve {keys [<at 0 0.3>,<at 0.6638 0.0872>,<at 1 0>]}
        BlueCurve {keys [<at 0 1>,<at 0.3872 1>,<at 1 1>]}
        GreenCurve {keys [<at 0 0.8261>,<at 0.3872 0.8655>,<at 1 0.1629>]}
        RedCurve {keys [<at 0 0.3062>,<at 0.3872 0.2665>,<at 1 0>]}
       }}
      CurveIndex=&Multiply_Float(A=$Particles.NormalizedAge, B=2)
     )
     Scale Factor=&MakeFloat4FromVectorAndFloat("Vector (RGB)"=&VectorFromFloat(Value=1), "Float (Alpha)"=1)
    )) {isSetParameters=true}
   /SetVariables(Particles.GlowScale:Vector2f=<20 20>) {isSetParameters=true}
   /SetVariables(Particles.GlowColor=&MultiplyLinearColor(
     Linear Color=&ColorFromCurve(ColorCurve=DataInterface() {propertyValues {
        AlphaCurve {keys [<at 0 0 RCIM_Cubic>,<at 0.0356 0.2>,<at 1 0 RCIM_Cubic>]}
        BlueCurve {keys [<at 0 1 RCIM_Cubic>,<at 1 1 RCIM_Cubic>]}
        GreenCurve {keys [<at 0 0.4764 RCIM_Cubic>,<at 1 0.1837 RCIM_Cubic>]}
        RedCurve {keys [<at 0 0.0859 RCIM_Cubic>,<at 1 0 RCIM_Cubic>]}
       }})
     Scale Factor=<1 1 1 0.5>
    )) {isSetParameters=true}
  ]
  rendererClasses [&NiagaraSpriteRendererProperties]
  renderers [
   {rendererClass=&NiagaraSpriteRendererProperties}
   {rendererClass=&NiagaraSpriteRendererProperties, rendererIndex=1}
  ]
 }
}
summary {
 systemName=NS_Emote_FlossYawn_SnowflakeBurst
 emitters [
  {bEnabled=false, emitterName=Minimal, rendererClasses [&NiagaraSpriteRendererProperties]}
  {bEnabled=false, emitterName=NE_First_Flare, rendererClasses [&NiagaraSpriteRendererProperties]}
  {bEnabled=false, emitterName=HeartsSparkles001, rendererClasses [&NiagaraSpriteRendererProperties]}
  {bEnabled=false, emitterName=DirectionalBurst, rendererClasses [&NiagaraSpriteRendererProperties]}
  {emitterName=Empty, rendererClasses []}
  {emitterName=Empty001, rendererClasses []}
  {emitterName=Empty002, rendererClasses []}
  {emitterName=Empty003, rendererClasses []}
  {emitterName=Empty004, rendererClasses []}
 ]
 userVariables [
  User.bIsFrontEnd {type=NiagaraBool}
  User.bIsFrontEndPreview {type=NiagaraBool}
 ]
}
